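# Niagara system for the NoiseClueGust player emote's big stage-style light effect, layering flashing light sprites around the character.
# Flashing ground lights spawn in rings through hand-timed bursts across roughly 21 seconds, while a smaller emitter loops delayed bursts higher up.
# Sprite size and alpha follow curves, with alpha also scaled by platform, quality level and time of day.
# An ambient emitter with SpawnGroup-switched layers and front-end adjustments is present but disabled.
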NiagaraSystem /BRCosmetics/Animation/Game/MainPlayer/Emotes/NoiseClueGust/FX/NS_Emote_NoiseGlueGustBigLight
dependencies {
 usedDataInterfaces [
  &NiagaraDataInterfaceSpriteRendererInfo
  &NiagaraDataInterfaceCurve
  &NiagaraDataInterfaceVector2DCurve
  &NiagaraDataInterfaceArrayFloat
  &NiagaraDataInterfacePlatformSet
 ]
 usedDynamicInputs [
  &MakeFloatFromVector2d
  &InvertBool
  &MakeLinearColor
  &Add_Float
 ]
 usedModules [
  &SystemState
  &BlendSpaceHelper
  &self:SystemSpawnScript.SystemScriptSource.SystemScriptGraph.NiagaraNodeAssignment_0.SetVariables_#1
  &self:SystemSpawnScript.SystemScriptSource.SystemScriptGraph.NiagaraNodeAssignment_7.SetVariables_#2
  &InterpolateOverTime
  &self:SystemSpawnScript.SystemScriptSource.SystemScriptGraph.NiagaraNodeAssignment_2.SetVariables_#3
 ]
}
emitters {
 Ambient_Condensed {
  bEnabled=false
  emitterUpdateScript [
   &EmitterState(Life Cycle Mode=::NewEnumerator1, Loop Duration=9.3, UseLoopDelay=false)
   &SpawnBurst_Instantaneous(Spawn Count=3, Spawn Group=3, Use Spawn Probability=false, UseLoopCountLimit=false) {module=SpawnBurst_Instantaneous001}
   &SpawnRate(Spawn Probability=0.75, SpawnGroup=3, SpawnRate=2.2, Use Spawn Probability=true) {module=SpawnRate003}
   &SpawnBurst_Instantaneous(Spawn Count=3, Spawn Group=4, Use Spawn Probability=false, UseLoopCountLimit=true) {module=SpawnBurst_Instantaneous002}
   &SpawnRate(Spawn Probability=0.75, SpawnGroup=4, SpawnRate=1.8, Use Spawn Probability=true) {module=SpawnRate004}
   &SpawnBurst_Instantaneous(Spawn Group=5, Use Spawn Probability=false, UseLoopCountLimit=false) {module=SpawnBurst_Instantaneous003}
   &SpawnRate(SpawnGroup=6, Use Spawn Probability=true) {module=SpawnRate005}
   &SpawnBurst_Instantaneous(Spawn Group=7, Use Spawn Probability=false, UseLoopCountLimit=false) {module=SpawnBurst_Instantaneous004}
   &SpawnBurst_Instantaneous(Use Spawn Probability=false, UseLoopCountLimit=false)
  ]
  particleSpawnScript [
   /SetVariables(Particles.VisibilityTag=&SpawnGroup()) {isSetParameters=true}
   &FN_Sprite_Attributes(Create Sprite Rotation Attribute=true, Write to Module Namespace=true)
   &InitializeParticle(
    Color=&MergeHelper_ColorSwitch(
     0=<rgba 0.9062 0.9062 0.9062 0.3>
     1=<rgba 1 1 0.9667 0.02>
     2=<rgba 1.15 1.1385 1.1385 0.1>
     3=<rgba 0.989 1 0.9588 0.11>
     4=<rgba 0.875 0.875 0.875 0.09>
     5=<rgba 0.6979 0.6979 0.6979 0.1>
     6=<rgba 1 1 1 0.3>
     7=<rgba 0.6979 0.6979 0.6979 0.2>
     Index=&SpawnGroup()
     Num=::NewEnumerator6
    )
    Lifetime=&MergeHelper_FloatSwitch(
     0=&RandomRangeFloat(Minimum=0.5, Maximum=1)
     1=&RandomRangeFloat(Minimum=2.5, Maximum=3.5)
     2=&RandomRangeFloat(Minimum=2.5, Maximum=3.5)
     3=&RandomRangeFloat(Minimum=5, Maximum=3)
     4=&RandomRangeFloat(Minimum=5, Maximum=3)
     5=9.3
     6=&RandomRangeFloat(Minimum=3, Maximum=5)
     7=9.3
     Index=&SpawnGroup()
     Num=::NewEnumerator6
    )
    Mesh Scale=&Multiply_VectorByFloat(
     Vector=&RandomRangeVector(Minimum=<1 1 2>, Maximum=<1 1 3>)
     Float=&RandomRangeFloat(Minimum=0.15, Maximum=0.3)
    )
    Mesh Scale Mode=::NewEnumerator4
    Position Offset=&MergeHelper_VectorSwitch(
     0=<0 0 -15>
     1=<0 0 72>
     2=&FN_FrontEndVector3Modifier(Gameplay=<0 0 150>, Front End=<0 0 110>)
     3=<0 0 0>
     4=<0 0 20>
     5=&MakeCustomVectorFromBool(
      Bool=&SetBoolByIntComparison(A=$Engine.QualityLevel, B=0, Comparison Type=::NewEnumerator4)
      True Vector=<0 0 110>
      False Vector=<0 0 70>
     )
     6=<0 0 0>
     7=<0 0 3>
     Index=&SpawnGroup()
     Num=::NewEnumerator6
    )
    Sprite Rotation Mode=::NewEnumerator2
    Sprite Size=&MergeHelper_Vector2Switch(
     0=&RandomRangeVector2D(Minimum=<26 150>, Maximum=<130 300>)
     1=&Vector2DFromFloat(Value=&RandomRangeFloat(Minimum=200, Maximum=300))
     2=&Vector2DFromFloat(Value=&RandomRangeFloat(Minimum=200, Maximum=300))
     3=&Vector2DFromFloat(Value=&RandomRangeFloat(Minimum=80, Maximum=190))
     4=&Vector2DFromFloat(Value=&RandomRangeFloat(Minimum=80, Maximum=190))
     5=&FN_FrontEndVector2Modifier(Front End:Vector2f=<213.3333 266.6667>, Gameplay:Vector2f=<266.6667 333.3333>)
     6=<0 0>
     7=<200 200>
     Index=&SpawnGroup()
     Num=::NewEnumerator6
    )
    Sprite Size Mode=::NewEnumerator4
    UsePositionOffset=true
    Write Material Random=true
   )
   /SetVariables(Particles.RaySide:NiagaraInt32=&RandomRangeInt(Minimum=0, Maximum=1)) {isSetParameters=true}
   &SubUVAnimation(
    Evaluation Type=::NewEnumerator0
    Output Alternate Renderer Binding=true
    "Sprite Renderer "=DataInterface() {propertyValues {SpriteRenderer=&self:Ambient_Condensed_0.NiagaraSpriteRendererProperties_2}}
    SubUV Animation Mode:ENiagara_SubUVLookupModeV2=::NewEnumerator2
    UseEndFrame=false
    UseStartFrame=false
   )
   &ShapeLocation(
    Hemisphere Distribution=<1 0.55>
    MaskSpawnGroup=false
    Non Uniform Scale=<1 1 0.35>
    Offset=&MaskVector(
     Mask Source=0
     Mask Parameter=?
     Mask=0
     Masked Vector=&MaskVector(Mask Source=1, Mask Parameter=$Particles.RaySide, Mask=0, Masked Vector=<60 -40 170>, Passthrough Vector=<-60 -40 170>)
     Passthrough Vector=<0 0 0>
    )
    Offset Mode=::NewEnumerator0
    Override Seed=false
    Sphere Radius=&MaskFloat(Mask Source=0, Mask Parameter=?, Mask=0, Masked Float=10, Passthrough Float=0)
    UseOwnerScale=true
   ) {module=ShapeLocation001}
   &ShapeLocation(
    Cylinder Height=&MaskFloat(
     Mask Source=0
     Mask Parameter=?
     Mask=1
     Masked Float=&FN_FrontEndFloatModifier(Gameplay=50, Front End=35)
     Passthrough Float=0
    )
    Cylinder Radius=&MaskFloat(
     Mask Source=0
     Mask Parameter=?
     Mask=1
     Masked Float=&FN_FrontEndFloatModifier(Gameplay=30, Front End=40)
     Passthrough Float=0
    )
    MaskSpawnGroup=false
    Non Uniform Scale=<1.5 1.25 1>
    Override Seed=false
    Shape Primitive=::NewEnumerator1
    Surface Only=false
    UseOwnerScale=true
   )
   &ShapeLocation(
    Cylinder Height=&MaskFloat(
     Mask Source=0
     Mask Parameter=?
     Mask=2
     Masked Float=&FN_FrontEndFloatModifier(Gameplay=85, Front End=35)
     Passthrough Float=0
    )
    Cylinder Radius=&MaskFloat(
     Mask Source=0
     Mask Parameter=?
     Mask=2
     Masked Float=&FN_FrontEndFloatModifier(Gameplay=30, Front End=40)
     Passthrough Float=0
    )
    MaskSpawnGroup=false
    Non Uniform Scale=<1.5 1.25 1>
    Override Seed=false
    Shape Primitive=::NewEnumerator1
    Surface Only=false
    UseOwnerScale=true
   ) {module=ShapeLocation002}
   &ShapeLocation(
    MaskSpawnGroup=false
    Non Uniform Scale=<1.25 1 1>
    Override Seed=false
    Ring Radius=&MaskFloat(Mask Source=0, Mask Parameter=?, Mask=3, Masked Float=30, Passthrough Float=0)
    Shape Primitive=::NewEnumerator4
    UseOwnerScale=true
   ) {module=ShapeLocation003}
   &ShapeLocation(
    Cylinder Height=&MaskFloat(Mask Source=0, Mask Parameter=?, Mask=4, Masked Float=25, Passthrough Float=0)
    Cylinder Radius=&MaskFloat(Mask Source=0, Mask Parameter=?, Mask=4, Masked Float=20, Passthrough Float=0)
    MaskSpawnGroup=false
    Non Uniform Scale=<1.25 1 1>
    Override Seed=false
    Shape Primitive=::NewEnumerator1
    Surface Only=false
    UseOwnerScale=true
   ) {module=ShapeLocation004}
   &AddVelocity(
    Module.Use Falloff Distance=false
    Origin Offset=<0 0 -50>
    Override Seed=false
    Velocity Mode=::NewEnumerator1
    Velocity Speed=&MaskFloat(Mask Source=0, Mask Parameter=?, Mask=0, Masked Float=0.01, Passthrough Float=0)
   )
   &AddVelocity(Velocity=&MergeHelper_VectorSwitch(
     0=&RandomRangeVector(
      Minimum=&MaskVector(Mask Source=1, Mask Parameter=$Particles.RaySide, Mask=0, Masked Vector=<-5 1 0>, Passthrough Vector=<5 1 0>)
      Maximum=<0.1 5 -5>
     )
     1=<0 0 0>
     2=<0 0 0>
     3=<0 0 3>
     4=<0 0 12>
     5=<0 0 0>
     6=<0 0 0>
     7=<0 0 0>
     Index=&SpawnGroup()
     Num=::NewEnumerator6
    )) {module=AddVelocity001}
   &CameraOffset(Camera Offset Amount=&MergeHelper_FloatSwitch(
     0=0
     1=-50
     2=-50
     3=&RandomRangeFloat(Minimum=-2, Maximum=25)
     4=&RandomRangeFloat(Minimum=-8, Maximum=15)
     5=-3
     Index=&SpawnGroup()
     Num=::NewEnumerator4
    ))
   /SetVariables(Particles.Glow Orientation:Vector3f=<0 0 1>) {isSetParameters=true}
  ]
  particleUpdateScript [
   &ParticleState()
   &ScaleSpriteSize(
    Scale Factor=&MergeHelper_Vector2Switch(
     0=&Vector2DFromCurve(
      Vector2Curve=DataInterface() {propertyValues {
        XCurve {
         defaultValue=0
         keys [
          {interpMode=RCIM_Cubic, tangentMode=RCTM_User, time=0, value=0}
          {arriveTangentWeight=0.1667, interpMode=RCIM_Cubic, tangentMode=RCTM_User, time=0.5, value=1}
          {arriveTangent=-3.3018, interpMode=RCIM_Cubic, leaveTangent=-3.3018, tangentMode=RCTM_User, time=0.8652, value=0.4388}
          <at 1 0>
         ]
        }
        YCurve {
         defaultValue=0
         keys [
          {interpMode=RCIM_Cubic, tangentMode=RCTM_User, time=0, value=0}
          {arriveTangent=2.4744, interpMode=RCIM_Cubic, leaveTangent=2.4744, tangentMode=RCTM_User, time=0.209, value=0.758}
          {arriveTangentWeight=0.1667, interpMode=RCIM_Cubic, tangentMode=RCTM_User, time=0.5, value=1}
          <at 1 0>
         ]
        }
       }}
      ScaleCurve=<1 1>
     )
     1=&Vector2DFromFloat(Value=&FloatFromCurve(FloatCurve=DataInterface() {propertyValues {Curve {keys [{tangentMode=RCTM_User, time=0, value=1}, {tangentMode=RCTM_User, time=1, value=1.2}]}}}))
     2=&Vector2DFromFloat(Value=&FloatFromCurve(FloatCurve=DataInterface() {propertyValues {Curve {keys [{tangentMode=RCTM_User, time=0, value=1}, {tangentMode=RCTM_User, time=1, value=1.2}]}}}))
     3=<1 1>
     4=&Vector2DFromFloat(Value=&FloatFromCurve(FloatCurve=DataInterface() {propertyValues {Curve {
          defaultValue=0
          keys [
           {interpMode=RCIM_Cubic, tangentMode=RCTM_User, time=0, value=0}
           {interpMode=RCIM_Cubic, tangentMode=RCTM_User, time=0.29, value=1}
           {arriveTangent=-1.4085, interpMode=RCIM_Cubic, leaveTangent=-1.4085, time=0.6493, value=0.8525}
           {tangentMode=RCTM_User, time=1, value=0}
          ]
         }}}))
     5=<1 1>
     6=<1 1>
     7=<1 1>
     Index=&SpawnGroup()
     Num=::NewEnumerator6
    )
    Scale Sprite Size Mode=::NewEnumerator2
   )
   &SubUVAnimation(
    End Frame Range Override=40
    Output Alternate Renderer Binding=true
    "Sprite Renderer "=DataInterface() {propertyValues {SpriteRenderer=&self:Ambient_Condensed_0.NiagaraSpriteRendererProperties_1}}
    SubUV Lookup Index Scale=0.6
    UseEndFrame=true
    UseStartFrame=true
   ) {module=SubUVAnimation001}
   &SubUVAnimation(
    End Frame Range Override=40
    Output Alternate Renderer Binding=true
    "Sprite Renderer "=DataInterface() {propertyValues {SpriteRenderer=&self:Ambient_Condensed_0.NiagaraSpriteRendererProperties_3}}
    SubUV Lookup Index Scale=0.6
    UseEndFrame=true
    UseStartFrame=true
   ) {module=SubUVAnimation002}
   &SubUVAnimation(
    Output Alternate Renderer Binding=true
    "Sprite Renderer "=DataInterface() {propertyValues {SpriteRenderer=&self:Ambient_Condensed_0.NiagaraSpriteRendererProperties_0}}
    UseEndFrame=false
    UseStartFrame=false
   ) {module=SubUVAnimation003}
   &SubUVAnimation(
    Output Alternate Renderer Binding=true
    "Sprite Renderer "=DataInterface() {propertyValues {SpriteRenderer=&self:Ambient_Condensed_0.NiagaraSpriteRendererProperties_4}}
    UseEndFrame=false
    UseStartFrame=false
   ) {module=SubUVAnimation004}
   &ScaleMeshSize(Scale Factor=&Multiply_VectorByFloat(
     Vector=&Multiply_VectorByFloat(Vector=$Engine.Owner.Scale, Float=&FN_FrontEndFloatModifier(Gameplay=0.9, Front End=0.8))
     Float=&FloatFromCurve(
      FloatCurve=DataInterface() {propertyValues {Curve {keys [
          {arriveTangent=0.0893, interpMode=RCIM_Cubic, leaveTangent=0.0893, tangentMode=RCTM_User, time=0, value=0.55}
          {arriveTangent=0.0158, interpMode=RCIM_Cubic, leaveTangent=0.0158, tangentMode=RCTM_User, time=1, value=1}
         ]}}}
      Scale Curve=&Multiply_Float(
       A=4
       B=&FloatFromCurve(
        FloatCurve=DataInterface() {propertyValues {Curve {keys [{interpMode=RCIM_Cubic, tangentMode=RCTM_User, time=0, value=0.4}, {interpMode=RCIM_Cubic, tangentMode=RCTM_User, time=0.2095, value=1}]}}}
        CurveIndex=$Emitter.Age
       )
      )
     )
    )) {module=ScaleMeshSize001}
   &UpdateMeshOrientation(
    Rotation Rate=&SelectFloatFromArray(
     Array Sampling Mode=::NewEnumerator0
     Direct Array Index=?
     Float Selection Array=DataInterface() {propertyValues {FloatData [-0.3,0.3,0.2,-0.2]}}
     Random Seed=0
    )
    Rotation Vector=<0 0 1>
   )
   /SetVariables(Particles.Glow Dyn:Vector4f=<1.7 1 1 1>) {isSetParameters=true}
   &GravityForce(Gravity=&MaskVector(
     Mask Source=0
     Mask Parameter=?
     Mask=4
     Masked Vector=&RandomRangeVector(Minimum=<0 0 8>, Maximum=<0 0 20>)
     Passthrough Vector=<0 0 0>
    ))
   &Drag(
    Drag=&MaskFloat(Mask Source=0, Mask Parameter=?, Mask=4, Masked Float=0.5, Passthrough Float=0)
    Rotational Drag:NiagaraFloat=&MaskFloat(Mask Source=0, Mask Parameter=?, Mask=4, Masked Float=1, Passthrough Float=0)
    Use Linear Drag=true
    Use Rotational Drag=true
   )
   &ScaleColor(
    Scale Alpha=&MergeHelper_FloatSwitch(
     0=&FloatFromCurve(
      FloatCurve=DataInterface() {propertyValues {Curve {
         defaultValue=0
         keys [
          {interpMode=RCIM_Cubic, tangentMode=RCTM_User, time=0, value=0}
          {interpMode=RCIM_Cubic, tangentMode=RCTM_User, time=0.5, value=1}
          {tangentMode=RCTM_User, time=1, value=0}
         ]
        }}}
      Scale Curve=&Multiply_Float(
       A=&MakeCustomFloatFromBool(
        Bool=&SetBoolByFloatComparison(
         A=&FN_PickFloatDependOnPlatform(
          TargetPlatform=DataInterface() {propertyValues {Platforms {
             qualityLevelMask=-30
             deviceProfileStates [
              {profileName=Windows, qualityLevelMask=0, setQualityLevelMask=2}
              {profileName=Mobile, qualityLevelMask=0, setQualityLevelMask=4294967295}
              {profileName=Mac, qualityLevelMask=0, setQualityLevelMask=2}
              {profileName=Linux, qualityLevelMask=0, setQualityLevelMask=2}
             ]
            }}}
          TargetPlatformActive=1
          TargetPlatformNoActive=0
         )
         B=0
        )
        True Float=0.5
        False Float=1
       )
       B=&Multiply_Float(
        A=&MakeCustomFloatFromBool(
         Bool=&SetBoolByIntComparison(A=$Engine.QualityLevel, B=0, Comparison Type=::NewEnumerator4)
         True Float=0.65
         False Float=1
        )
        B=&Multiply_Float(
         A=&RandomRangeFloat(Minimum=0.23, Maximum=1.1)
         B=&FNTimeOfDayScalar(Night Scale=0.4, Morning Scale=0.75, Evening Scale=0.75)
        )
       )
      )
     )
     1=&FloatFromCurve(
      FloatCurve=DataInterface() {propertyValues {Curve {keys [
          {interpMode=RCIM_Cubic, tangentMode=RCTM_User, time=0, value=0}
          {interpMode=RCIM_Cubic, tangentMode=RCTM_User, time=0.2766, value=0.3}
          {tangentMode=RCTM_User, time=1, value=0}
         ]}}}
      Scale Curve=&Multiply_Float(
       A=&MakeCustomFloatFromBool(
        Bool=&SetBoolByFloatComparison(
         A=&FN_PickFloatDependOnPlatform(
          TargetPlatform=DataInterface() {propertyValues {Platforms {
             qualityLevelMask=-30
             deviceProfileStates [
              {profileName=Windows, qualityLevelMask=0, setQualityLevelMask=2}
              {profileName=Mobile, qualityLevelMask=0, setQualityLevelMask=4294967295}
              {profileName=Mac, qualityLevelMask=0, setQualityLevelMask=2}
              {profileName=Linux, qualityLevelMask=0, setQualityLevelMask=2}
             ]
            }}}
          TargetPlatformActive=1
          TargetPlatformNoActive=0
         )
         B=0
        )
        True Float=0.5
        False Float=1
       )
       B=&Multiply_Float(
        A=&MakeCustomFloatFromBool(
         Bool=&SetBoolByIntComparison(A=$Engine.QualityLevel, B=0, Comparison Type=::NewEnumerator4)
         True Float=0.65
         False Float=1
        )
        B=&Multiply_Float(
         A=&FN_FrontEndFloatModifier(Gameplay=0.22, Front End=0.12)
         B=&FNTimeOfDayScalar(Night Scale=0.5, Morning Scale=0.75, Evening Scale=0.75)
        )
       )
      )
     )
     2=&FloatFromCurve(
      FloatCurve=DataInterface() {propertyValues {Curve {keys [
          {interpMode=RCIM_Cubic, tangentMode=RCTM_User, time=0, value=0}
          {interpMode=RCIM_Cubic, tangentMode=RCTM_User, time=0.2766, value=0.1}
          {tangentMode=RCTM_User, time=1, value=0}
         ]}}}
      Scale Curve=&Multiply_Float(
       A=&MakeCustomFloatFromBool(
        Bool=&SetBoolByFloatComparison(
         A=&FN_PickFloatDependOnPlatform(
          TargetPlatform=DataInterface() {propertyValues {Platforms {
             qualityLevelMask=-30
             deviceProfileStates [
              {profileName=Windows, qualityLevelMask=0, setQualityLevelMask=2}
              {profileName=Mobile, qualityLevelMask=0, setQualityLevelMask=4294967295}
              {profileName=Mac, qualityLevelMask=0, setQualityLevelMask=2}
              {profileName=Linux, qualityLevelMask=0, setQualityLevelMask=2}
             ]
            }}}
          TargetPlatformActive=1
          TargetPlatformNoActive=0
         )
         B=0
        )
        True Float=0.5
        False Float=1
       )
       B=&Multiply_Float(
        A=&MakeCustomFloatFromBool(
         Bool=&SetBoolByIntComparison(A=$Engine.QualityLevel, B=0, Comparison Type=::NewEnumerator4)
         True Float=0.65
         False Float=1
        )
        B=&Multiply_Float(
         A=&FN_FrontEndFloatModifier(Gameplay=0.3, Front End=0.12)
         B=&FNTimeOfDayScalar(Night Scale=0.5, Morning Scale=0.75, Evening Scale=0.75)
        )
       )
      )
     )
     3=&FloatFromCurve(
      FloatCurve=DataInterface() {propertyValues {Curve {
         defaultValue=1
         keys [
          {interpMode=RCIM_Cubic, tangentMode=RCTM_User, time=0, value=0}
          {interpMode=RCIM_Cubic, tangentMode=RCTM_User, time=0.2807, value=1.0588}
          {tangentMode=RCTM_User, time=1, value=0}
         ]
        }}}
      Scale Curve=&Multiply_Float(
       A=&MakeCustomFloatFromBool(
        Bool=&SetBoolByFloatComparison(
         A=&FN_PickFloatDependOnPlatform(
          TargetPlatform=DataInterface() {propertyValues {Platforms {
             qualityLevelMask=-29
             deviceProfileStates [
              {profileName=Windows, qualityLevelMask=0, setQualityLevelMask=3}
              {profileName=Linux, qualityLevelMask=0, setQualityLevelMask=3}
              {profileName=Mac, qualityLevelMask=0, setQualityLevelMask=3}
              {profileName=IOS, qualityLevelMask=0, setQualityLevelMask=4294967295}
             ]
            }}}
          TargetPlatformActive=1
          TargetPlatformNoActive=0
         )
         B=0
        )
        True Float=0.5
        False Float=1
       )
       B=&Multiply_Float(
        A=&MakeCustomFloatFromBool(
         Bool=&SetBoolByIntComparison(A=$Engine.QualityLevel, B=0, Comparison Type=::NewEnumerator4)
         True Float=0.65
         False Float=1
        )
        B=&Multiply_Float(
         A=&FN_FrontEndFloatModifier(Gameplay=0.65, Front End=0.3)
         B=&FNTimeOfDayScalar(Night Scale=0.6, Morning Scale=0.8, Evening Scale=0.8)
        )
       )
      )
     )
     4=&FloatFromCurve(
      FloatCurve=DataInterface() {propertyValues {Curve {
         defaultValue=0
         keys [
          {arriveTangent=13.0062, interpMode=RCIM_Cubic, leaveTangent=13.0062, tangentMode=RCTM_User, time=0, value=0}
          {interpMode=RCIM_Cubic, tangentMode=RCTM_User, time=0.1852, value=1}
          {arriveTangent=-0.7697, interpMode=RCIM_Cubic, leaveTangent=-0.7697, tangentMode=RCTM_User, time=0.7051, value=0.8679}
          {arriveTangent=-1.9584, interpMode=RCIM_Cubic, leaveTangent=-1.9584, tangentMode=RCTM_User, time=1, value=0}
         ]
        }}}
      Scale Curve=&Multiply_Float(
       A=&MakeCustomFloatFromBool(
        Bool=&SetBoolByFloatComparison(
         A=&FN_PickFloatDependOnPlatform(
          TargetPlatform=DataInterface() {propertyValues {Platforms {
             qualityLevelMask=-30
             deviceProfileStates [
              {profileName=Windows, qualityLevelMask=0, setQualityLevelMask=2}
              {profileName=Mobile, qualityLevelMask=0, setQualityLevelMask=4294967295}
              {profileName=Mac, qualityLevelMask=0, setQualityLevelMask=2}
              {profileName=Linux, qualityLevelMask=0, setQualityLevelMask=2}
             ]
            }}}
          TargetPlatformActive=1
          TargetPlatformNoActive=0
         )
         B=0
        )
        True Float=0.5
        False Float=1
       )
       B=&Multiply_Float(
        A=&MakeCustomFloatFromBool(
         Bool=&SetBoolByIntComparison(A=$Engine.QualityLevel, B=0, Comparison Type=::NewEnumerator4)
         True Float=0.33
         False Float=1
        )
        B=&Multiply_Float(
         A=&FN_FrontEndFloatModifier(
          Gameplay=&RandomRangeFloat(Minimum=0.6, Maximum=0.83)
          Front End=&FloatFromCurve(FloatCurve=DataInterface() {propertyValues {Curve {
              defaultValue=0.3
              keys [
               {interpMode=RCIM_Cubic, tangentMode=RCTM_User, time=0, value=0.4}
               {arriveTangent=-1.1537, interpMode=RCIM_Cubic, leaveTangent=-1.1537, tangentMode=RCTM_User, time=1, value=0}
              ]
             }}})
         )
         B=&FNTimeOfDayScalar(Night Scale=0.6, Morning Scale=0.8, Evening Scale=0.8)
        )
       )
      )
     )
     5=&FloatFromCurve(
      FloatCurve=DataInterface() {propertyValues {Curve {keys [<at 0 0>,<at 0.15 1>,<at 0.85 1>,<at 1 0>]}}}
      Scale Curve=&Multiply_Float(
       A=&MakeCustomFloatFromBool(
        Bool=&SetBoolByFloatComparison(
         A=&FN_PickFloatDependOnPlatform(
          TargetPlatform=DataInterface() {propertyValues {Platforms {
             qualityLevelMask=-29
             deviceProfileStates [
              {profileName=Windows, qualityLevelMask=0, setQualityLevelMask=3}
              {profileName=Linux, qualityLevelMask=0, setQualityLevelMask=3}
              {profileName=Mac, qualityLevelMask=0, setQualityLevelMask=3}
              {profileName=IOS, qualityLevelMask=0, setQualityLevelMask=4294967295}
             ]
            }}}
          TargetPlatformActive=1
          TargetPlatformNoActive=0
         )
         B=0
        )
        True Float=0.2
        False Float=1
       )
       B=&FN_FrontEndFloatModifier(Gameplay=1, Front End=0.6)
      )
     )
     6=&FloatFromCurve(FloatCurve=DataInterface() {propertyValues {Curve {
         defaultValue=1
         keys [
          {interpMode=RCIM_Cubic, tangentMode=RCTM_User, time=0, value=0}
          {interpMode=RCIM_Cubic, tangentMode=RCTM_User, time=0.5, value=1}
          {tangentMode=RCTM_User, time=1, value=0}
         ]
        }}})
     7=&FloatFromCurve(
      FloatCurve=DataInterface() {propertyValues {Curve {
         defaultValue=1
         keys [
          {interpMode=RCIM_Cubic, tangentMode=RCTM_User, time=0, value=0.15}
          {interpMode=RCIM_Cubic, tangentMode=RCTM_User, time=0.2002, value=1}
          {arriveTangent=-1.0624, interpMode=RCIM_Cubic, leaveTangent=-1.0624, time=0.8198, value=1}
          {tangentMode=RCTM_User, time=1, value=0.1503}
         ]
        }}}
      Scale Curve=0.7
     )
     Index=&SpawnGroup()
     Num=::NewEnumerator6
    )
    Scale RGB=&MergeHelper_VectorSwitch(
     0=<1 1 1>
     1=&Multiply_VectorByFloat(Vector=$System.LightColor, Float=&RandomRangeFloat(Minimum=0.5, Maximum=1))
     2=&Multiply_VectorByFloat(Vector=$System.LightColor, Float=&RandomRangeFloat(Minimum=0.5, Maximum=1))
     3=$System.LightColor
     4=$System.LightColor
     5=$System.LightColor
     6=$System.LightColor
     7=$System.LightColor
     Index=&SpawnGroup()
     Num=::NewEnumerator6
    )
    ScaleA=true
    ScaleRGB=true
   )
   &CurlNoiseForce(
    Bias Noise=true
    Bias Noise Field:Vector3f=<0 1 0>
    Noise Strength=&MaskFloat(Mask Source=0, Mask Parameter=?, Mask=0, Masked Float=10, Passthrough Float=0)
    Pan Noise=false
   )
   &SolveForcesAndVelocity(Clamp Velocity=false, Limit Acceleration=false)
   &BlendSpace(Rotation Contribution=0)
  ]
  rendererClasses [&NiagaraSpriteRendererProperties, &NiagaraMeshRendererProperties]
  renderers [
   {rendererClass=&NiagaraSpriteRendererProperties}
   {rendererClass=&NiagaraSpriteRendererProperties, rendererIndex=1}
   {rendererClass=&NiagaraSpriteRendererProperties, rendererIndex=2}
   {rendererClass=&NiagaraSpriteRendererProperties, rendererIndex=3}
   {rendererClass=&NiagaraSpriteRendererProperties, rendererIndex=4}
   {rendererClass=&NiagaraSpriteRendererProperties, rendererIndex=5}
   {rendererClass=&NiagaraMeshRendererProperties, rendererIndex=6}
   {rendererClass=&NiagaraSpriteRendererProperties, rendererIndex=7}
  ]
 }
 Back_Light {rendererClasses [], renderers []}
 Flashing_Lights_GRoundBig {
  emitterUpdateScript [
   &EmitterState()
   /Niagara/Modules/Emitter/SpawnByTrigger(
    Spawn Count:NiagaraInt32=&RandomRangeInt(Minimum=4, Maximum=6)
    Trigger:NiagaraBool=&SetBoolByIntComparison(
     A=&MakeIntFromFloat(
      Float=&FloatFromCurve(
       FloatCurve=DataInterface() {propertyValues {Curve {keys [<at 0 1 RCIM_Constant>,<at 0.511 0 RCIM_Constant>,<at 1.02 1 RCIM_Constant>,<at 1.436 0 RCIM_Constant>,<at 1.83 1 RCIM_Constant>,<at 2.3793 0 RCIM_Constant>,<at 2.87 1 RCIM_Constant>,<at 3.3769 0 RCIM_Constant>,<at 3.92 1 RCIM_Constant>,<at 4.1176 0 RCIM_Constant>,<at 4.35 1 RCIM_Constant>,<at 5.2631 0 RCIM_Constant>,<at 5.75 1 RCIM_Constant>,<at 6.4238 0 RCIM_Constant>,<at 6.79 1 RCIM_Constant>,<at 6.83 1 RCIM_Constant>,<at 7.0056 0 RCIM_Constant>,<at 7.25 1 RCIM_Constant>,<at 7.8519 0 RCIM_Constant>,<at 8.3 1 RCIM_Constant>,<at 8.8568 0 RCIM_Constant>,<at 9.33 1 RCIM_Constant>,<at 9.7878 0 RCIM_Constant>,<at 10.16 1 RCIM_Constant>,<at 10.6975 0 RCIM_Constant>,<at 11.2 1 RCIM_Constant>,<at 11.4802 0 RCIM_Constant>,<at 11.95 1 RCIM_Constant>,<at 12.3715 0 RCIM_Constant>,<at 12.88 0.9961 RCIM_Constant>]}}}
       CurveIndex=$Emitter.Age
      )
      Conversion Mode=::NewEnumerator1
     )
     B=0
     Comparison Type=::NewEnumerator0
    )
    TriggerIntervalSwitch=false
   ) {enabled=false}
   &SpawnByTrigger(
    Spawn Count:NiagaraInt32=&RandomRangeInt(Minimum=4, Maximum=6)
    Trigger:NiagaraBool=&SetBoolByIntComparison(
     A=&MakeIntFromFloat(
      Float=&FloatFromCurve(
       FloatCurve=DataInterface() {propertyValues {Curve {keys [<at 0 1 RCIM_Constant>,<at 0.4 0 RCIM_Constant>,<at 1.428 1 RCIM_Constant>,<at 1.828 0 RCIM_Constant>,<at 2.142 1 RCIM_Constant>,<at 2.4769 0 RCIM_Constant>,<at 2.856 1 RCIM_Constant>,<at 3.256 0 RCIM_Constant>,<at 4.284 1 RCIM_Constant>,<at 4.684 0 RCIM_Constant>,<at 4.998 1 RCIM_Constant>,<at 5.3493 0 RCIM_Constant>,<at 5.712 1 RCIM_Constant>,<at 6.112 0 RCIM_Constant>,<at 7.14 1 RCIM_Constant>,<at 7.54 0 RCIM_Constant>,<at 8.568 1 RCIM_Constant>,<at 8.968 0 RCIM_Constant>,<at 9.282 1 RCIM_Constant>,<at 9.5671 0 RCIM_Constant>,<at 9.996 1 RCIM_Constant>,<at 10.396 0 RCIM_Constant>,<at 11.424 0.9961 RCIM_Constant>,<at 11.824 0 RCIM_Constant>,<at 12.852 1 RCIM_Constant>]}}}
       CurveIndex=$Emitter.Age
      )
      Conversion Mode=::NewEnumerator1
     )
     B=0
     Comparison Type=::NewEnumerator0
    )
    Trigger Interval=0.759
    TriggerIntervalSwitch=false
   ) {enabled=false, module=SpawnByTrigger001}
   &SpawnBurst_Instantaneous(Spawn Count=5, Spawn Group=1, Use Spawn Probability=false, UseLoopCountLimit=false)
   &SpawnBurst_Instantaneous(Spawn Count=5, Spawn Group=1, Spawn Time=1.02, Use Spawn Probability=false, UseLoopCountLimit=false) {module=SpawnBurst_Instantaneous001}
   &SpawnBurst_Instantaneous(Spawn Count=4, Spawn Group=1, Spawn Time=1.83, Use Spawn Probability=false, UseLoopCountLimit=false) {module=SpawnBurst_Instantaneous002}
   &SpawnBurst_Instantaneous(Spawn Count=5, Spawn Group=1, Spawn Time=2.87, Use Spawn Probability=false, UseLoopCountLimit=false) {module=SpawnBurst_Instantaneous003}
   &SpawnBurst_Instantaneous(Spawn Count=5, Spawn Group=1, Spawn Time=3.89, Use Spawn Probability=false, UseLoopCountLimit=false) {module=SpawnBurst_Instantaneous004}
   &SpawnBurst_Instantaneous(Spawn Count=4, Spawn Group=1, Spawn Time=4.7, Use Spawn Probability=false, UseLoopCountLimit=false) {module=SpawnBurst_Instantaneous005}
   &SpawnBurst_Instantaneous(Spawn Count=5, Spawn Group=1, Spawn Time=5.74, Use Spawn Probability=false, UseLoopCountLimit=false) {module=SpawnBurst_Instantaneous006}
   &SpawnBurst_Instantaneous(Spawn Count=5, Spawn Group=1, Spawn Time=6.76, Use Spawn Probability=false, UseLoopCountLimit=false) {module=SpawnBurst_Instantaneous007}
   &SpawnBurst_Instantaneous(Spawn Count=4, Spawn Group=1, Spawn Time=7.57, Use Spawn Probability=false, UseLoopCountLimit=false) {module=SpawnBurst_Instantaneous008}
   &SpawnBurst_Instantaneous(Spawn Count=5, Spawn Group=1, Spawn Time=8.61, Use Spawn Probability=false, UseLoopCountLimit=false) {module=SpawnBurst_Instantaneous009}
   &SpawnBurst_Instantaneous(Spawn Count=5, Spawn Group=1, Spawn Time=9.63, Use Spawn Probability=false, UseLoopCountLimit=false) {module=SpawnBurst_Instantaneous011}
   &SpawnBurst_Instantaneous(Spawn Count=4, Spawn Group=1, Spawn Time=10.44, Use Spawn Probability=false, UseLoopCountLimit=false) {module=SpawnBurst_Instantaneous012}
   &SpawnBurst_Instantaneous(Spawn Count=5, Spawn Group=1, Spawn Time=11.48, Use Spawn Probability=false, UseLoopCountLimit=false) {module=SpawnBurst_Instantaneous013}
   &SpawnBurst_Instantaneous(Spawn Count=3, Spawn Group=1, Spawn Time=11.99, Use Spawn Probability=false, UseLoopCountLimit=false) {module=SpawnBurst_Instantaneous015}
   &SpawnBurst_Instantaneous(Spawn Count=4, Spawn Group=1, Spawn Time=12.5, Use Spawn Probability=false, UseLoopCountLimit=false) {module=SpawnBurst_Instantaneous014}
   &SpawnBurst_Instantaneous(Spawn Count=3, Spawn Group=1, Spawn Time=13.31, Use Spawn Probability=false, UseLoopCountLimit=false) {module=SpawnBurst_Instantaneous016}
   &SpawnBurst_Instantaneous(Spawn Count=5, Spawn Group=1, Spawn Time=14.35, Use Spawn Probability=false, UseLoopCountLimit=false) {module=SpawnBurst_Instantaneous017}
   &SpawnBurst_Instantaneous(Spawn Count=3, Spawn Group=1, Spawn Time=14.86, Use Spawn Probability=false, UseLoopCountLimit=false) {module=SpawnBurst_Instantaneous018}
   &SpawnBurst_Instantaneous(Spawn Count=4, Spawn Group=1, Spawn Time=15.37, Use Spawn Probability=false, UseLoopCountLimit=false) {module=SpawnBurst_Instantaneous019}
   &SpawnBurst_Instantaneous(Spawn Count=3, Spawn Group=1, Spawn Time=16.18, Use Spawn Probability=false, UseLoopCountLimit=false) {module=SpawnBurst_Instantaneous020}
   &SpawnBurst_Instantaneous(Spawn Count=5, Spawn Group=1, Spawn Time=17.22, Use Spawn Probability=false, UseLoopCountLimit=false) {module=SpawnBurst_Instantaneous023}
   &SpawnBurst_Instantaneous(Spawn Count=3, Spawn Group=1, Spawn Time=17.73, Use Spawn Probability=false, UseLoopCountLimit=false) {module=SpawnBurst_Instantaneous010}
   &SpawnBurst_Instantaneous(Spawn Count=4, Spawn Group=1, Spawn Time=18.24, Use Spawn Probability=false, UseLoopCountLimit=false) {module=SpawnBurst_Instantaneous021}
   &SpawnBurst_Instantaneous(Spawn Count=3, Spawn Group=1, Spawn Time=19.05, Use Spawn Probability=false, UseLoopCountLimit=false) {module=SpawnBurst_Instantaneous022}
   &SpawnBurst_Instantaneous(Spawn Count=5, Spawn Group=1, Spawn Time=20.09, Use Spawn Probability=false, UseLoopCountLimit=false) {module=SpawnBurst_Instantaneous027}
   &SpawnBurst_Instantaneous(Spawn Count=3, Spawn Group=1, Spawn Time=20.6, Use Spawn Probability=false, UseLoopCountLimit=false) {module=SpawnBurst_Instantaneous024}
   &SpawnBurst_Instantaneous(Spawn Count=4, Spawn Group=1, Spawn Time=21.11, Use Spawn Probability=false, UseLoopCountLimit=false) {module=SpawnBurst_Instantaneous025}
  ]
  particleSpawnScript [
   &InitializeParticle(
    Color=<rgba 0.8482 0.8345 1 0.4>
    Lifetime=&RandomRangeFloat(Minimum=0.75, Maximum=0.8)
    Sprite Size=&RandomRangeVector2D(Minimum=<80 325>, Maximum=<120 400>)
    Sprite Size Mode=::NewEnumerator4
    UsePositionOffset=false
    Write Material Random=true
   )
   /Niagara/Modules/Spawn/Location/V2/GridLocation(
    Module.Jitter Inside Grid Cells=false
    Module.Override Random Seed=false
    Offset:Vector3f=&MaskVector(Mask Source=0, Mask Parameter=?, Mask=0, Masked Vector=<6 -50 18>, Passthrough Vector=<-6 -50 18>)
    XYZ Dimensions:Vector3f=<10 1 10>
   ) {enabled=false}
   &ShapeLocation(
    MaskSpawnGroup=false
    Offset=<0 0 20>
    Offset Coordinate Space=::Local
    Offset Mode=::NewEnumerator0
    Override Seed=false
    Ring / Disc Distribution Mode:ENiagaraSphereDistributionMode=::NewEnumerator1
    Ring Radius=75
    Rotation Angle=0.5
    Rotation Axis=<0 0 1>
    Rotation Mode=::NewEnumerator1
    Shape Primitive=::NewEnumerator4
    U Position=&ReturnNormalizedExecIndex(Normalized Index Scale=0.5)
    UseOwnerScale=true
   )
   &SubUVAnimation(
    Evaluation Type=::NewEnumerator0
    Output Alternate Renderer Binding=true
    "Sprite Renderer "=DataInterface() {propertyValues {SpriteRenderer=&self:Flashing_Lights_GRoundBig_0.NiagaraSpriteRendererProperties_2}}
    SubUV Animation Mode:ENiagara_SubUVLookupModeV2=::NewEnumerator2
    UseEndFrame=false
    UseStartFrame=false
   )
   &AddVelocity(
    Velocity=&RandomRangeVector(
     Minimum=&MaskVector(Mask Source=1, Mask Parameter=$Emitter.SpawnGroup, Mask=0, Masked Vector=<-0.5 0.5 4>, Passthrough Vector=<-0.5 0.5 4>)
     Maximum=<-0.5 2 5>
    )
    Velocity Speed Scale=0.01
   )
   &AddVelocity(
    Velocity=&Multiply_VectorByFloat(Vector=$Particles.ShapeLocation.ShapeTangent, Float=1)
    Velocity Speed Scale=0.01
   ) {module=AddVelocity001}
   &ApplyInitialForces()
   &KillParticles(Kill Particles=&SetBoolByFloatComparison(
     A=&UniformAOrBFloat(A=0, Distribution Weight=0.4, Evaluation Type=::NewEnumerator0, Recalculate Random Each Loop=true, Override Seed=false)
     B=0.5
    )) {enabled=false}
   /SetVariables(Particles.LightSize:Vector2f=<60 60>) {isSetParameters=true, module=SetVariables_#4}
  ]
  particleUpdateScript [
   &ParticleState()
   &ScaleSpriteSize(
    Scale Factor=&Vector2DFromCurve(
     Vector2Curve=DataInterface() {propertyValues {
       XCurve {
        defaultValue=0
        keys [
         {interpMode=RCIM_Cubic, tangentMode=RCTM_User, time=0, value=0}
         {arriveTangentWeight=0.1667, interpMode=RCIM_Cubic, tangentMode=RCTM_User, time=0.4409, value=0.975}
         {arriveTangent=-3.3018, interpMode=RCIM_Cubic, leaveTangent=-3.3018, tangentMode=RCTM_User, time=0.7922, value=0.425}
         <at 1 0>
        ]
       }
       YCurve {
        defaultValue=0
        keys [
         {interpMode=RCIM_Cubic, tangentMode=RCTM_User, time=0, value=0}
         {arriveTangent=2.4744, interpMode=RCIM_Cubic, leaveTangent=2.4744, tangentMode=RCTM_User, time=0.209, value=0.758}
         {arriveTangentWeight=0.1667, interpMode=RCIM_Cubic, tangentMode=RCTM_User, time=0.5, value=1}
         <at 1 0>
        ]
       }
      }}
     ScaleCurve=<1.4 0.8>
    )
    Scale Sprite Size Mode=::NewEnumerator2
   )
   &CurlNoiseForce(Bias Noise=true, Bias Noise Field:Vector3f=<0 1 0>, Noise Frequency=0, Noise Strength=0, Pan Noise=false)
   &SolveForcesAndVelocity(Clamp Velocity=false, Limit Acceleration=false)
   &ScaleColor(
    Scale Alpha=&FloatFromCurve(
     FloatCurve=DataInterface() {propertyValues {Curve {
        defaultValue=0
        keys [
         {interpMode=RCIM_Cubic, tangentMode=RCTM_User, time=0, value=0}
         {interpMode=RCIM_Cubic, tangentMode=RCTM_User, time=0.1374, value=1}
         {tangentMode=RCTM_User, time=1, value=0}
        ]
       }}}
     Scale Curve=&Multiply_Float(
      A=&MakeCustomFloatFromBool(
       Bool=&SetBoolByFloatComparison(
        A=&FN_PickFloatDependOnPlatform(
         TargetPlatform=DataInterface() {propertyValues {Platforms {
            qualityLevelMask=-30
            deviceProfileStates [
             {profileName=Windows, qualityLevelMask=0, setQualityLevelMask=2}
             {profileName=Mobile, qualityLevelMask=0, setQualityLevelMask=4294967295}
             {profileName=Mac, qualityLevelMask=0, setQualityLevelMask=2}
             {profileName=Linux, qualityLevelMask=0, setQualityLevelMask=2}
            ]
           }}}
         TargetPlatformActive=1
         TargetPlatformNoActive=0
        )
        B=0
       )
       True Float=0.5
       False Float=1
      )
      B=&Multiply_Float(
       A=&MakeCustomFloatFromBool(
        Bool=&SetBoolByIntComparison(A=$Engine.QualityLevel, B=0, Comparison Type=::NewEnumerator4)
        True Float=0.65
        False Float=1
       )
       B=&Multiply_Float(
        A=&RandomRangeFloat(Minimum=0.23, Maximum=1.1)
        B=&FNTimeOfDayScalar(Night Scale=0.4, Morning Scale=0.75, Evening Scale=0.75)
       )
      )
     )
    )
    ScaleA=true
    ScaleRGB=true
   )
   /SetVariables(
    Particles.FlareColor=&Multiply_LinearColorByFloat(Linear Color=$Particles.Color, Float=0.7)
    Particles.FlashSphereColor:LinearColor=&Multiply_LinearColorByFloat(Linear Color=$Particles.Color, Float=0.7)
   ) {isSetParameters=true}
  ]
  rendererClasses [&NiagaraSpriteRendererProperties]
  renderers [
   {rendererClass=&NiagaraSpriteRendererProperties}
   {rendererClass=&NiagaraSpriteRendererProperties, rendererIndex=1}
   {rendererClass=&NiagaraSpriteRendererProperties, rendererIndex=2}
  ]
 }
 Flashing_Lights_Small {
  emitterUpdateScript [
   &EmitterState(Delay First Loop Only=true, Life Cycle Mode=::NewEnumerator1, Loop Delay=0.51, Loop Duration=1.435, UseLoopDelay=true)
   &SpawnBurst_Instantaneous(Spawn Count=5, Use Spawn Probability=false, UseLoopCountLimit=false)
   &SpawnBurst_Instantaneous(Spawn Count=4, Spawn Group=1, Spawn Time=0.3587, Use Spawn Probability=false, UseLoopCountLimit=false) {enabled=false, module=SpawnBurst_Instantaneous001}
   /Niagara/Modules/Spawn/Location/SpawnParticlesInGrid(X Count=5, Y Count=0, Z Count=3) {enabled=false}
   &SpawnParticlesInGrid(Spawn Group=1, X Count=5, Y Count=0, Z Count=3) {enabled=false, module=SpawnParticlesInGrid001}
  ]
  particleSpawnScript [
   &InitializeParticle(
    Color=<rgba 0.8482 0.8345 1 0.6>
    Lifetime=&RandomRangeFloat(Minimum=0.55, Maximum=0.6)
    Sprite Size=&RandomRangeVector2D(Minimum=<60 225>, Maximum=<80 250>)
    Sprite Size Mode=::NewEnumerator4
    UsePositionOffset=false
    Write Material Random=true
   )
   &GridLocation(
    Module.Jitter Inside Grid Cells=false
    Module.Override Random Seed=false
    Offset:Vector3f=&MaskVector(Mask Source=0, Mask Parameter=?, Mask=0, Masked Vector=<6 -50 18>, Passthrough Vector=<-6 -50 18>)
    XYZ Dimensions:Vector3f=<10 1 10>
   ) {enabled=false}
   &SubUVAnimation(
    Evaluation Type=::NewEnumerator0
    Output Alternate Renderer Binding=true
    "Sprite Renderer "=DataInterface() {propertyValues {SpriteRenderer=&self:NiagaraEmitter_0.NiagaraSpriteRendererProperties_2}}
    SubUV Animation Mode:ENiagara_SubUVLookupModeV2=::NewEnumerator2
    UseEndFrame=false
    UseStartFrame=false
   )
   &ShapeLocation(
    MaskSpawnGroup=false
    Offset=&MaskVectorBySpawnGroup(Masked Vector=<0 0 180>, Passthrough Vector=<0 0 180>)
    Offset Coordinate Space=::Local
    Offset Mode=::NewEnumerator0
    Override Seed=false
    Ring / Disc Distribution Mode:ENiagaraSphereDistributionMode=::NewEnumerator1
    Ring Radius=75
    Rotation Angle=0.5
    Rotation Axis=<0 0 1>
    Rotation Mode=::NewEnumerator1
    Shape Primitive=::NewEnumerator4
    U Position=&ReturnNormalizedExecIndex(Normalized Index Scale=0.5)
    UseOwnerScale=true
   )
   &AddVelocity(
    Velocity=&RandomRangeVector(
     Minimum=&MaskVector(Mask Source=1, Mask Parameter=$Emitter.SpawnGroup, Mask=0, Masked Vector=<-2 2 -5>, Passthrough Vector=<-2 2 -5>)
     Maximum=&MaskVector(Mask Source=1, Mask Parameter=$Emitter.SpawnGroup, Mask=0, Masked Vector=<2 2 -5>, Passthrough Vector=<2 2 -5>)
    )
    Velocity Speed Scale=0.01
   )
   &AddVelocity(
    Velocity=&Multiply_VectorByFloat(Vector=$Particles.ShapeLocation.ShapeTangent, Float=-1)
    Velocity Speed Scale=0.01
   ) {module=AddVelocity001}
   &ApplyInitialForces()
   /SetVariables(Particles.LightSize:Vector2f=<50 50>) {isSetParameters=true, module=SetVariables_#4}
  ]
  particleUpdateScript [
   &ParticleState()
   &ScaleSpriteSize(
    Scale Factor=&Vector2DFromCurve(
     Vector2Curve=DataInterface() {propertyValues {
       XCurve {
        defaultValue=0
        keys [
         {interpMode=RCIM_Cubic, tangentMode=RCTM_User, time=0, value=0}
         {arriveTangentWeight=0.1667, interpMode=RCIM_Cubic, tangentMode=RCTM_User, time=0.5, value=1}
         {arriveTangent=-3.3018, interpMode=RCIM_Cubic, leaveTangent=-3.3018, tangentMode=RCTM_User, time=0.8652, value=0.4388}
         <at 1 0>
        ]
       }
       YCurve {
        defaultValue=0
        keys [
         {interpMode=RCIM_Cubic, tangentMode=RCTM_User, time=0, value=0}
         {arriveTangent=2.4744, interpMode=RCIM_Cubic, leaveTangent=2.4744, tangentMode=RCTM_User, time=0.209, value=0.758}
         {arriveTangentWeight=0.1667, interpMode=RCIM_Cubic, tangentMode=RCTM_User, time=0.5, value=1}
         <at 1 0>
        ]
       }
      }}
     ScaleCurve=<1 1>
    )
    Scale Sprite Size Mode=::NewEnumerator2
   )
   &CurlNoiseForce(Bias Noise=true, Bias Noise Field:Vector3f=<0 1 0>, Noise Frequency=0, Noise Strength=0, Pan Noise=false)
   &SolveForcesAndVelocity(Clamp Velocity=false, Limit Acceleration=false)
   &ScaleColor(
    Scale Alpha=&FloatFromCurve(
     FloatCurve=DataInterface() {propertyValues {Curve {
        defaultValue=0
        keys [
         {interpMode=RCIM_Cubic, tangentMode=RCTM_User, time=0, value=0}
         {interpMode=RCIM_Cubic, tangentMode=RCTM_User, time=0.1374, value=1}
         {tangentMode=RCTM_User, time=1, value=0}
        ]
       }}}
     Scale Curve=&Multiply_Float(
      A=&MakeCustomFloatFromBool(
       Bool=&SetBoolByFloatComparison(
        A=&FN_PickFloatDependOnPlatform(
         TargetPlatform=DataInterface() {propertyValues {Platforms {
            qualityLevelMask=-30
            deviceProfileStates [
             {profileName=Windows, qualityLevelMask=0, setQualityLevelMask=2}
             {profileName=Mobile, qualityLevelMask=0, setQualityLevelMask=4294967295}
             {profileName=Mac, qualityLevelMask=0, setQualityLevelMask=2}
             {profileName=Linux, qualityLevelMask=0, setQualityLevelMask=2}
            ]
           }}}
         TargetPlatformActive=1
         TargetPlatformNoActive=0
        )
        B=0
       )
       True Float=0.5
       False Float=1
      )
      B=&Multiply_Float(
       A=&MakeCustomFloatFromBool(
        Bool=&SetBoolByIntComparison(A=$Engine.QualityLevel, B=0, Comparison Type=::NewEnumerator4)
        True Float=0.65
        False Float=1
       )
       B=&Multiply_Float(
        A=&RandomRangeFloat(Minimum=0.23, Maximum=1.1)
        B=&FNTimeOfDayScalar(Night Scale=0.4, Morning Scale=0.75, Evening Scale=0.75)
       )
      )
     )
    )
    ScaleA=true
    ScaleRGB=true
   )
   /SetVariables(
    Particles.FlareColor=&Multiply_LinearColorByFloat(Linear Color=$Particles.Color, Float=0.7)
    Particles.FlashSphereColor:LinearColor=&Multiply_LinearColorByFloat(Linear Color=$Particles.Color, Float=0.65)
   ) {isSetParameters=true}
  ]
  rendererClasses [&NiagaraSpriteRendererProperties]
  renderers [
   {rendererClass=&NiagaraSpriteRendererProperties}
   {rendererClass=&NiagaraSpriteRendererProperties, rendererIndex=1}
   {rendererClass=&NiagaraSpriteRendererProperties, rendererIndex=2}
  ]
 }
 GroundGlowBassSnare {
  emitterUpdateScript [
   &EmitterState()
   &SpawnByTrigger(
    Spawn Group=4
    Trigger:NiagaraBool=&SetBoolByFloatComparison(A=$System.BassSpawn, B=0.7)
    TriggerIntervalSwitch=false
   ) {module=SpawnByTrigger005}
   &SpawnByTrigger(
    Spawn Group=3
    Trigger:NiagaraBool=&SetBoolByFloatComparison(A=$System.BassSpawn, B=0.7)
    TriggerIntervalSwitch=false
   ) {module=SpawnByTrigger004}
   &SpawnByTrigger(
    Spawn Count=2
    Trigger:NiagaraBool=&SetBoolByFloatComparison(A=$System.BassSpawn, B=0.7)
    Trigger Interval=0.15
    TriggerIntervalSwitch=false
   )
   &SpawnByTrigger(
    Spawn Group=1
    Trigger:NiagaraBool=&SetBoolByFloatComparison(A=$System.SnareSpawn, B=0.7)
    TriggerIntervalSwitch=false
   ) {module=SpawnByTrigger001}
   &SpawnByTrigger(
    Spawn Group=2
    Trigger:NiagaraBool=&SetBoolByFloatComparison(A=$System.SnareSpawn, B=0.7)
    TriggerIntervalSwitch=false
   ) {module=SpawnByTrigger003}
   &SpawnByTrigger(
    Spawn Group=2
    Trigger:NiagaraBool=&SetBoolByFloatComparison(A=$System.BassSpawn, B=0.7)
    Trigger Interval=0.3
    TriggerCountSwitch=false
    TriggerIntervalSwitch=true
   ) {module=SpawnByTrigger002}
  ]
  particleSpawnScript [
   &InitializeParticle(
    Color=&MaskLinearColorBySpawnGroup(
     Masked Linear Color=<rgba 0.8482 0.8345 1 0.65>
     Passthrough Linear Color=&MaskLinearColorBySpawnGroup(
      Masked Linear Color=<rgba 0.8482 0.8345 1 0.65>
      Passthrough Linear Color=&MaskLinearColorBySpawnGroup(Masked Linear Color=<rgba 0.8482 0.8345 1 0.65>, Passthrough Linear Color=<rgba 0.8482 0.8345 1 1>, Spawn Group Mask=4)
      Spawn Group Mask=3
     )
    )
    Lifetime=&Multiply_Float(
     A=&MergeHelper_FloatSwitch(0=0.8, 1=0.6, 2=0.7, 3=0.7, 4=0.75, Index=$Emitter.SpawnGroup, Num=::NewEnumerator3)
     B=&RandomRangeFloat(Minimum=0.85, Maximum=0.9)
    )
    Mesh Scale=<2 2 1>
    Mesh Scale Mode=::NewEnumerator4
    Position Offset=<0 0 5>
    Sprite Size Mode=::NewEnumerator3
    Uniform Sprite Size=&Multiply_Float(
     A=&MergeHelper_FloatSwitch(
      0=&RandomRangeFloat(Minimum=100, Maximum=90)
      1=112
      2=112
      3=75
      4=55
      Index=$Emitter.SpawnGroup
      Num=::NewEnumerator3
     )
     B=&RandomRangeFloat(Minimum=0.75, Maximum=0.85)
    )
    UsePositionOffset=true
    Write Material Random=true
   )
   /SetVariables(Particles.DecalSize:Vector3f=<150 150 150>) {isSetParameters=true}
   /SetVariables(
    Particles.Glow DecalSize:Vector3f=<700 700 700>
    Particles.Glow Position:NiagaraPosition=&AddVectorToPosition(Position=/Niagara/DynamicInputs/Helpers/SimulationPosition(), Vector=<-15 -15 0>)
    Particles.VisibilityTag=&MergeHelper_IntSwitch(0=0, 1=1, 2=2, 3=0, 4=0, 5=?, Index=$Emitter.SpawnGroup, Num=::NewEnumerator3)
   ) {isSetParameters=true}
   &SpriteFacingAndAlignment(Sprite Facing=<0 0 1>, WriteAlignment=true, WriteFacing=true)
  ]
  particleUpdateScript [
   &ParticleState()
   &SpriteRotationRate(Rotation Rate=&RandomRangeFloat(Minimum=50, Maximum=60))
   &ScaleSpriteSize(
    Scale Sprite Size Mode=::NewEnumerator0
    Uniform Scale Factor=&MaskFloatBySpawnGroup(
     Masked Float=&FloatFromCurve(
      FloatCurve=DataInterface() {propertyValues {Curve {
         defaultValue=0
         keys [
          {arriveTangent=4.1785, interpMode=RCIM_Cubic, leaveTangent=4.1785, tangentMode=RCTM_User, time=0, value=0.2}
          {arriveTangent=1.1663, interpMode=RCIM_Cubic, leaveTangent=1.1663, tangentMode=RCTM_User, time=0.2042, value=0.7}
          {arriveTangent=0.0224, interpMode=RCIM_Cubic, leaveTangent=0.0224, tangentMode=RCTM_User, time=1, value=1}
         ]
        }}}
      Scale Curve=2.2
     )
     Passthrough Float=&MaskFloatBySpawnGroup(
      Masked Float=&FloatFromCurve(
       FloatCurve=DataInterface() {propertyValues {Curve {
          defaultValue=0
          keys [
           {arriveTangent=4.1785, interpMode=RCIM_Cubic, leaveTangent=4.1785, tangentMode=RCTM_User, time=0, value=0.2}
           {arriveTangent=1.1663, interpMode=RCIM_Cubic, leaveTangent=1.1663, tangentMode=RCTM_User, time=0.2042, value=0.7}
           {arriveTangent=0.0224, interpMode=RCIM_Cubic, leaveTangent=0.0224, tangentMode=RCTM_User, time=1, value=1}
          ]
         }}}
       Scale Curve=2.2
      )
      Passthrough Float=&MaskFloatBySpawnGroup(
       Masked Float=&FloatFromCurve(
        FloatCurve=DataInterface() {propertyValues {Curve {
           defaultValue=0
           keys [
            {arriveTangent=4.1785, interpMode=RCIM_Cubic, leaveTangent=4.1785, tangentMode=RCTM_User, time=0, value=0.2}
            {arriveTangent=1.1663, interpMode=RCIM_Cubic, leaveTangent=1.1663, tangentMode=RCTM_User, time=0.2042, value=0.7}
            {arriveTangent=0.0224, interpMode=RCIM_Cubic, leaveTangent=0.0224, tangentMode=RCTM_User, time=1, value=1}
           ]
          }}}
        Scale Curve=2.2
       )
       Passthrough Float=&FloatFromCurve(
        FloatCurve=DataInterface() {propertyValues {Curve {
           defaultValue=0
           keys [
            {arriveTangent=4.1785, interpMode=RCIM_Cubic, leaveTangent=4.1785, tangentMode=RCTM_User, time=0, value=0.2}
            {arriveTangent=1.1179, interpMode=RCIM_Cubic, leaveTangent=1.1179, tangentMode=RCTM_User, time=0.2042, value=0.6228}
            {arriveTangent=0.0224, interpMode=RCIM_Cubic, leaveTangent=0.0224, tangentMode=RCTM_User, time=1, value=1}
           ]
          }}}
        Scale Curve=2.1
       )
       Spawn Group Mask=4
      )
      Spawn Group Mask=3
     )
    )
   )
   &SolveForcesAndVelocity(Clamp Velocity=false, Limit Acceleration=false)
   &ScaleColor(
    Scale Alpha=&Multiply_Float(
     A=&FloatFromCurve(
      FloatCurve=DataInterface() {propertyValues {Curve {
         defaultValue=0.9847
         keys [
          {arriveTangent=-0.6538, interpMode=RCIM_Cubic, leaveTangent=-0.6538, tangentMode=RCTM_User, time=0.1, value=1.025}
          {arriveTangent=-0.46, interpMode=RCIM_Cubic, leaveTangent=-0.46, tangentMode=RCTM_User, time=1, value=0}
         ]
        }}}
      Scale Curve=&MaskFloatBySpawnGroup(
       Masked Float=0.165
       Passthrough Float=&MaskFloatBySpawnGroup(
        Masked Float=0.165
        Passthrough Float=&MaskFloatBySpawnGroup(
         Masked Float=0.165
         Passthrough Float=&FloatFromCurve(FloatCurve=DataInterface() {propertyValues {Curve {
             defaultValue=1
             keys [
              {arriveTangent=4.1785, interpMode=RCIM_Cubic, leaveTangent=4.1785, tangentMode=RCTM_User, time=0, value=0}
              {arriveTangent=0.7558, interpMode=RCIM_Cubic, leaveTangent=0.7558, tangentMode=RCTM_User, time=0.2, value=0.6228}
              {arriveTangent=0.0224, interpMode=RCIM_Cubic, leaveTangent=0.0224, tangentMode=RCTM_User, time=1, value=1}
             ]
            }}})
         Spawn Group Mask=4
        )
        Spawn Group Mask=3
       )
      )
     )
     B=&FNTimeOfDayScalar(Night Scale=0.8, Morning Scale=0.8, Evening Scale=0.8)
    )
    ScaleA=true
    ScaleRGB=true
   )
   &BlendSpace(Rotation Contribution=0)
   /SetVariables(Particles.Glow DecalColor:LinearColor=&MakeLinearColorFromVectorAndFloat(
     "Vector (RGB)"=&MakeVectorFromLinearColorRGB(Color=<rgba 0.8482 0.8345 1 1>)
     "Float (Alpha)"=&Multiply_Float(
      A=&FloatFromCurve(
       FloatCurve=DataInterface() {propertyValues {Curve {
          defaultValue=0
          keys [{interpMode=RCIM_Cubic, tangentMode=RCTM_User, time=0.2, value=0.5}, {interpMode=RCIM_Cubic, tangentMode=RCTM_User, time=1, value=0}]
         }}}
       Scale Curve=&Multiply_Float(A=0.17, B=&FNTimeOfDayScalar(Night Scale=0.66, Morning Scale=0.8, Evening Scale=0.8))
      )
      B=1
     )
    )) {isSetParameters=true}
   /SetVariables(Particles.LightDiffuseScale=10) {isSetParameters=true}
   &Light_Attributes(
    Color=1
    Diffuse Scale=1
    Enabled=1
    Falloff Exponent=1
    Light Diffuse Scale=20
    Light Radius=75
    Position=1
    Position to Offset:NiagaraPosition=$Engine.Emitter.SimulationPosition
    Radius=1
    Scale Alpha:NiagaraFloat=&FloatFromCurve(
     FloatCurve=DataInterface() {propertyValues {Curve {keys [
         {interpMode=RCIM_Cubic, tangentMode=RCTM_User, time=0, value=1}
         <at 0.8 0 RCIM_Constant>
         <at 1.02 1 RCIM_Cubic>
         <at 1.82 0 RCIM_Constant>
         <at 1.83 1 RCIM_Cubic>
         <at 2.63 0 RCIM_Constant>
         <at 2.87 1 RCIM_Cubic>
         <at 3.67 0 RCIM_Constant>
         <at 3.89 1 RCIM_Cubic>
         <at 4.69 0 RCIM_Constant>
         <at 4.7 1 RCIM_Cubic>
         <at 5.5 0 RCIM_Constant>
         <at 5.74 1 RCIM_Cubic>
         <at 6.54 0 RCIM_Constant>
         <at 6.76 1 RCIM_Cubic>
         <at 7.56 0 RCIM_Constant>
         <at 7.57 1 RCIM_Cubic>
         <at 8.37 0 RCIM_Constant>
         <at 8.61 1 RCIM_Cubic>
         <at 9.41 0 RCIM_Constant>
         <at 9.63 1 RCIM_Cubic>
         <at 10.43 0 RCIM_Constant>
         <at 10.44 1 RCIM_Cubic>
         <at 11.24 0 RCIM_Constant>
         <at 11.48 1 RCIM_Cubic>
         <at 12.28 0 RCIM_Constant>
         <at 12.5 1 RCIM_Cubic>
         <at 13.3 0 RCIM_Constant>
         <at 13.31 1 RCIM_Cubic>
         <at 14.11 0 RCIM_Constant>
         {interpMode=RCIM_Cubic, tangentMode=RCTM_User, time=14.35, value=1}
         <at 14.859 0.375 RCIM_Constant>
         {interpMode=RCIM_Cubic, tangentMode=RCTM_User, time=14.86, value=1}
         <at 15.369 0.375 RCIM_Constant>
         {interpMode=RCIM_Cubic, tangentMode=RCTM_User, time=15.37, value=1}
         <at 16.17 0 RCIM_Constant>
         <at 16.18 1 RCIM_Cubic>
         <at 16.98 0 RCIM_Constant>
         {arriveTangent=0.5068, interpMode=RCIM_Cubic, leaveTangent=0.5068, time=17.22, value=1}
         <at 17.72 0.375 RCIM_Constant>
         <at 17.73 1 RCIM_Cubic>
         <at 18.23 0.375 RCIM_Constant>
         {interpMode=RCIM_Cubic, tangentMode=RCTM_User, time=18.24, value=1}
         <at 19.04 0 RCIM_Constant>
         <at 19.05 1 RCIM_Cubic>
         <at 19.85 0 RCIM_Constant>
         <at 20.09 1 RCIM_Cubic>
         <at 20.89 0 RCIM_Constant>
         <at 21.11 1 RCIM_Cubic>
         <at 21.91 0 RCIM_Constant>
        ]}}}
     CurveIndex=$Engine.System.Age
    )
    ScaleA=true
    ScaleRGB=true
    Specular Scale=1
    UsePositionOffset=true
   )
  ]
  rendererClasses [&NiagaraDecalRendererProperties, &NiagaraSpriteRendererProperties]
  renderers [
   {rendererClass=&NiagaraDecalRendererProperties, rendererIndex=0}
   {rendererClass=&NiagaraSpriteRendererProperties, rendererIndex=1}
   {rendererClass=&NiagaraSpriteRendererProperties, rendererIndex=2}
  ]
 }
 Ground_Light {rendererClasses [], renderers []}
 Light_Left {rendererClasses [], renderers []}
 Light_Right {rendererClasses [], renderers []}
}
summary {
 systemName=NS_Emote_NoiseGlueGustBigLight
 emitters [
  {
   bEnabled=false
   emitterName=Ambient_Condensed
   rendererClasses [&NiagaraSpriteRendererProperties, &NiagaraMeshRendererProperties]
  }
  {emitterName=Flashing_Lights_GRoundBig, rendererClasses [&NiagaraSpriteRendererProperties]}
  {emitterName=Flashing_Lights_Small, rendererClasses [&NiagaraSpriteRendererProperties]}
  {
   emitterName=GroundGlowBassSnare
   rendererClasses [&NiagaraDecalRendererProperties, &NiagaraSpriteRendererProperties]
  }
  {emitterName=Light_Right, rendererClasses []}
  {emitterName=Back_Light, rendererClasses []}
  {emitterName=Light_Left, rendererClasses []}
  {emitterName=Ground_Light, rendererClasses []}
 ]
 userVariables [
  User.bIsFrontEnd {type=NiagaraBool}
  User.bIsFrontEndPreview {type=NiagaraBool}
  User.Fade {type=NiagaraFloat, defaultValue=0, description=""}
  User.Pyrotechnics {type=NiagaraFloat, defaultValue=0.5, description=""}
  User.SquareRingColor {type=LinearColor, description="", defaultValue {struct=LinearColor, value=<rgba 0 0.3611 0.7083 1>}}
 ]
}
