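# Niagara system for the NoiseClueGust emote that plays pale lavender-white flashing lights with pulsing ground glows.
# Two active flash emitters burst sprites on rings at raised offsets on a shared 1.435-second beat with staggered timing; a third is disabled.
# Decal and GroundGlow emitters pair rotating, growing sprites with fading decals.
# Brightness adapts to time of day, plus platform and quality for flashes and front-end versus gameplay for glow decals.
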NiagaraSystem /BRCosmetics/Animation/Game/MainPlayer/Emotes/NoiseClueGust/FX/NS_Emote_NoiseGlueGustSmallLight
dependencies {
 usedDataInterfaces [
  &NiagaraDataInterfaceSpriteRendererInfo
  &NiagaraDataInterfaceCurve
  &NiagaraDataInterfaceVector2DCurve
  &NiagaraDataInterfacePlatformSet
 ]
 usedDynamicInputs [
  &MakeFloatFromVector2d
  &InvertBool
  &MakeLinearColor
  &NormalizeFloat
  &Add_Float
 ]
 usedModules [
  &SystemState
  &BlendSpaceHelper
  &self:SystemSpawnScript.SystemScriptSource.SystemScriptGraph.NiagaraNodeAssignment_0.SetVariables_#1
  &self:SystemSpawnScript.SystemScriptSource.SystemScriptGraph.NiagaraNodeAssignment_7.SetVariables_#2
  &InterpolateOverTime
  &self:SystemSpawnScript.SystemScriptSource.SystemScriptGraph.NiagaraNodeAssignment_2.SetVariables_#3
 ]
}
emitters {
 Back_Light {rendererClasses [], renderers []}
 Decal {
  emitterUpdateScript [
   &EmitterState(
    Delay First Loop Only=true
    Life Cycle Mode=::NewEnumerator1
    Loop Behavior=::NewEnumerator2
    Loop Count=4
    Loop Delay=0.7175
    Loop Duration=1.435
    UseLoopDelay=true
   )
   &SpawnBurst_Instantaneous(Use Spawn Probability=false, UseLoopCountLimit=false)
  ]
  particleSpawnScript [
   &InitializeParticle(
    Color=<rgba 0.8482 0.8345 1 1>
    Lifetime=0.65
    Mesh Scale=<2.5 2.5 1>
    Mesh Scale Mode=::NewEnumerator4
    Position Offset=<0 0 3>
    Sprite Size Mode=::NewEnumerator3
    Uniform Sprite Size=65
    UsePositionOffset=true
    Write Material Random=true
   )
   /SetVariables(Particles.Glow DecalSize:Vector3f=<550 500 500>) {isSetParameters=true, module=SetVariables_#4}
   /SetVariables(Particles.DecalSize:Vector3f=<170 170 170>) {isSetParameters=true, module=SetVariables_#5}
   &SpriteFacingAndAlignment(Sprite Facing=<0 0 1>, WriteAlignment=true, WriteFacing=true)
  ]
  particleUpdateScript [
   &ParticleState()
   &SolveForcesAndVelocity(Clamp Velocity=false, Limit Acceleration=false)
   &SpriteRotationRate(Rotation Rate=&RandomRangeFloat(Minimum=120, Maximum=150))
   &ScaleSpriteSize(
    Scale Sprite Size Mode=::NewEnumerator0
    Uniform Scale Factor=&FloatFromCurve(
     FloatCurve=DataInterface() {propertyValues {Curve {
        defaultValue=0
        keys [
         {arriveTangent=5.6398, interpMode=RCIM_Cubic, leaveTangent=5.6398, tangentMode=RCTM_User, time=0, value=0}
         {arriveTangent=0.9164, interpMode=RCIM_Cubic, leaveTangent=0.9164, tangentMode=RCTM_User, time=0.2042, value=0.6228}
         {arriveTangent=0.2266, interpMode=RCIM_Cubic, leaveTangent=0.2266, tangentMode=RCTM_User, time=1, value=1}
        ]
       }}}
     Scale Curve=2.1
    )
   )
   &ScaleColor(
    Scale Alpha=&Multiply_Float(
     A=&FloatFromCurve(
      FloatCurve=DataInterface() {propertyValues {Curve {
         defaultValue=0.9847
         keys [
          {arriveTangent=-1.0638, interpMode=RCIM_Cubic, leaveTangent=-1.0638, tangentMode=RCTM_User, time=0.05, value=1.025}
          {arriveTangent=-0.1561, interpMode=RCIM_Cubic, leaveTangent=-0.1561, tangentMode=RCTM_User, time=1, value=0}
         ]
        }}}
      Scale Curve=0.7
     )
     B=1
    )
    ScaleA=true
    ScaleRGB=true
   )
   &BlendSpace(Rotation Contribution=0)
   /SetVariables(Particles.Glow DecalColor:LinearColor=&MakeLinearColorFromVectorAndFloat(
     "Vector (RGB)"=&MakeVectorFromLinearColorRGB(Color=<rgba 1 1 1 1>)
     "Float (Alpha)"=&Multiply_Float(
      A=&FloatFromCurve(
       FloatCurve=DataInterface() {propertyValues {Curve {keys [
           {arriveTangent=-2.4342, interpMode=RCIM_Cubic, leaveTangent=-2.4342, tangentMode=RCTM_User, time=0, value=1}
           {arriveTangent=-0.3806, interpMode=RCIM_Cubic, leaveTangent=-0.3806, tangentMode=RCTM_User, time=1, value=0}
          ]}}}
       Scale Curve=&Multiply_Float(
        A=&FN_FrontEndFloatModifier(Gameplay=0.13, Front End=0.15)
        B=&FNTimeOfDayScalar(Night Scale=0.8, Morning Scale=0.8, Evening Scale=0.8)
       )
      )
      B=1
     )
    )) {isSetParameters=true, module=SetVariables_#6}
  ]
  rendererClasses [&NiagaraDecalRendererProperties, &NiagaraSpriteRendererProperties]
  renderers [
   {rendererClass=&NiagaraDecalRendererProperties, rendererIndex=0}
   {rendererClass=&NiagaraSpriteRendererProperties, rendererIndex=1}
  ]
 }
 Flashing_Lights_Small {
  emitterUpdateScript [
   &EmitterState(Delay First Loop Only=true, Life Cycle Mode=::NewEnumerator1, Loop Duration=1.435, UseLoopDelay=true)
   &SpawnBurst_Instantaneous(Spawn Count=3, Use Spawn Probability=false, UseLoopCountLimit=false)
   &SpawnBurst_Instantaneous(Spawn Count=4, Spawn Group=1, Spawn Time=0.3587, Use Spawn Probability=false, UseLoopCountLimit=false) {module=SpawnBurst_Instantaneous001}
   &SpawnBurst_Instantaneous(Spawn Count=3, Spawn Time=1.0762, Use Spawn Probability=false, UseLoopCountLimit=false) {module=SpawnBurst_Instantaneous002}
   /Niagara/Modules/Spawn/Location/SpawnParticlesInGrid(X Count=5, Y Count=0, Z Count=3) {enabled=false}
   &SpawnParticlesInGrid(Spawn Group=1, X Count=5, Y Count=0, Z Count=3) {enabled=false, module=SpawnParticlesInGrid001}
  ]
  particleSpawnScript [
   &InitializeParticle(
    Color=<rgba 0.8482 0.8345 1 0.6>
    Lifetime=&RandomRangeFloat(Minimum=0.7, Maximum=0.75)
    Sprite Size=&RandomRangeVector2D(Minimum=<60 225>, Maximum=<80 250>)
    Sprite Size Mode=::NewEnumerator4
    UsePositionOffset=false
    Write Material Random=true
   )
   /Niagara/Modules/Spawn/Location/V2/GridLocation(
    Module.Jitter Inside Grid Cells=false
    Module.Override Random Seed=false
    Offset:Vector3f=&MaskVector(Mask Source=0, Mask Parameter=?, Mask=0, Masked Vector=<6 -50 18>, Passthrough Vector=<-6 -50 18>)
    XYZ Dimensions:Vector3f=<10 1 10>
   ) {enabled=false}
   &SubUVAnimation(
    Evaluation Type=::NewEnumerator0
    Output Alternate Renderer Binding=true
    "Sprite Renderer "=DataInterface() {propertyValues {SpriteRenderer=&self:Flashing_Lights_Small_0.NiagaraSpriteRendererProperties_2}}
    SubUV Animation Mode:ENiagara_SubUVLookupModeV2=::NewEnumerator2
    UseEndFrame=false
    UseStartFrame=false
   )
   &ShapeLocation(
    MaskSpawnGroup=false
    Offset=&MaskVectorBySpawnGroup(Masked Vector=<0 0 180>, Passthrough Vector=<0 0 160>)
    Offset Coordinate Space=::Local
    Offset Mode=::NewEnumerator0
    Override Seed=false
    Radius Position:NiagaraFloat=&RandomRangeFloat(Minimum=0.95, Maximum=1)
    Ring / Disc Distribution Mode:ENiagaraSphereDistributionMode=::NewEnumerator1
    Ring Radius=75
    Rotation Angle=0.5
    Rotation Axis=<0 0 1>
    Rotation Mode=::NewEnumerator1
    Shape Primitive=::NewEnumerator4
    U Position=&ReturnNormalizedExecIndex(Normalized Index Scale=0.5)
    UseOwnerScale=true
   )
   &AddVelocity(
    Velocity=&RandomRangeVector(
     Minimum=&MaskVector(Mask Source=1, Mask Parameter=$Emitter.SpawnGroup, Mask=0, Masked Vector=<2 1 -5>, Passthrough Vector=<2 1 -5>)
     Maximum=&MaskVector(Mask Source=1, Mask Parameter=$Emitter.SpawnGroup, Mask=0, Masked Vector=<-2 2 -5>, Passthrough Vector=<-2 2 -5>)
    )
    Velocity Speed Scale=0.01
   )
   &AddVelocity(
    Velocity=&Multiply_VectorByFloat(Vector=$Particles.ShapeLocation.ShapeTangent, Float=-1)
    Velocity Speed Scale=0.01
   ) {module=AddVelocity001}
   &ApplyInitialForces()
   /SetVariables(Particles.LightSize:Vector2f=<50 50>) {isSetParameters=true, module=SetVariables_#7}
  ]
  particleUpdateScript [
   &ParticleState()
   &ScaleSpriteSize(
    Scale Factor=&Vector2DFromCurve(
     Vector2Curve=DataInterface() {propertyValues {
       XCurve {
        defaultValue=0
        keys [
         {interpMode=RCIM_Cubic, tangentMode=RCTM_User, time=0, value=0}
         {arriveTangentWeight=0.1667, interpMode=RCIM_Cubic, tangentMode=RCTM_User, time=0.5, value=1}
         {arriveTangent=-3.3018, interpMode=RCIM_Cubic, leaveTangent=-3.3018, tangentMode=RCTM_User, time=0.8652, value=0.4388}
         <at 1 0>
        ]
       }
       YCurve {
        defaultValue=0
        keys [
         {interpMode=RCIM_Cubic, tangentMode=RCTM_User, time=0, value=0}
         {arriveTangent=2.4744, interpMode=RCIM_Cubic, leaveTangent=2.4744, tangentMode=RCTM_User, time=0.209, value=0.758}
         {arriveTangentWeight=0.1667, interpMode=RCIM_Cubic, tangentMode=RCTM_User, time=0.5, value=1}
         <at 1 0>
        ]
       }
      }}
     ScaleCurve=<1 1>
    )
    Scale Sprite Size Mode=::NewEnumerator2
   )
   &CurlNoiseForce(Bias Noise=true, Bias Noise Field:Vector3f=<0 1 0>, Noise Frequency=0, Noise Strength=0, Pan Noise=false)
   &SolveForcesAndVelocity(Clamp Velocity=false, Limit Acceleration=false)
   &ScaleColor(
    Scale Alpha=&FloatFromCurve(
     FloatCurve=DataInterface() {propertyValues {Curve {
        defaultValue=0
        keys [
         {interpMode=RCIM_Cubic, tangentMode=RCTM_User, time=0, value=0}
         {interpMode=RCIM_Cubic, tangentMode=RCTM_User, time=0.1374, value=1}
         {tangentMode=RCTM_User, time=1, value=0}
        ]
       }}}
     Scale Curve=&Multiply_Float(
      A=&MakeCustomFloatFromBool(
       Bool=&SetBoolByFloatComparison(
        A=&FN_PickFloatDependOnPlatform(
         TargetPlatform=DataInterface() {propertyValues {Platforms {
            qualityLevelMask=-30
            deviceProfileStates [
             {profileName=Windows, qualityLevelMask=0, setQualityLevelMask=2}
             {profileName=Mobile, qualityLevelMask=0, setQualityLevelMask=4294967295}
             {profileName=Mac, qualityLevelMask=0, setQualityLevelMask=2}
             {profileName=Linux, qualityLevelMask=0, setQualityLevelMask=2}
            ]
           }}}
         TargetPlatformActive=1
         TargetPlatformNoActive=0
        )
        B=0
       )
       True Float=0.5
       False Float=1
      )
      B=&Multiply_Float(
       A=&MakeCustomFloatFromBool(
        Bool=&SetBoolByIntComparison(A=$Engine.QualityLevel, B=0, Comparison Type=::NewEnumerator4)
        True Float=0.65
        False Float=1
       )
       B=&Multiply_Float(
        A=&RandomRangeFloat(Minimum=0.23, Maximum=1.1)
        B=&FNTimeOfDayScalar(Night Scale=0.4, Morning Scale=0.75, Evening Scale=0.75)
       )
      )
     )
    )
    ScaleA=true
    ScaleRGB=true
   )
   /SetVariables(
    Particles.FlareColor=&Multiply_LinearColorByFloat(Linear Color=$Particles.Color, Float=0.7)
    Particles.FlashSphereColor:LinearColor=&Multiply_LinearColorByFloat(Linear Color=$Particles.Color, Float=0.65)
   ) {isSetParameters=true, module=SetVariables_#8}
  ]
  rendererClasses [&NiagaraSpriteRendererProperties]
  renderers [
   {rendererClass=&NiagaraSpriteRendererProperties}
   {rendererClass=&NiagaraSpriteRendererProperties, rendererIndex=1}
   {rendererClass=&NiagaraSpriteRendererProperties, rendererIndex=2}
  ]
 }
 Flashing_Lights_Small001 {
  emitterUpdateScript [
   &EmitterState(Delay First Loop Only=true, Life Cycle Mode=::NewEnumerator1, Loop Delay=0.7175, Loop Duration=1.435, UseLoopDelay=true)
   &SpawnBurst_Instantaneous(Spawn Count=5, Use Spawn Probability=false, UseLoopCountLimit=false)
   &SpawnBurst_Instantaneous(Spawn Count=4, Spawn Group=1, Spawn Time=0.3587, Use Spawn Probability=false, UseLoopCountLimit=false) {enabled=false, module=SpawnBurst_Instantaneous001}
   &SpawnParticlesInGrid(X Count=5, Y Count=0, Z Count=3) {enabled=false}
   &SpawnParticlesInGrid(Spawn Group=1, X Count=5, Y Count=0, Z Count=3) {enabled=false, module=SpawnParticlesInGrid001}
  ]
  particleSpawnScript [
   &InitializeParticle(
    Color=<rgba 0.8482 0.8345 1 0.6>
    Lifetime=&RandomRangeFloat(Minimum=0.7, Maximum=0.75)
    Sprite Size=&RandomRangeVector2D(Minimum=<60 225>, Maximum=<80 250>)
    Sprite Size Mode=::NewEnumerator4
    UsePositionOffset=false
    Write Material Random=true
   )
   &GridLocation(
    Module.Jitter Inside Grid Cells=false
    Module.Override Random Seed=false
    Offset:Vector3f=&MaskVector(Mask Source=0, Mask Parameter=?, Mask=0, Masked Vector=<6 -50 18>, Passthrough Vector=<-6 -50 18>)
    XYZ Dimensions:Vector3f=<10 1 10>
   ) {enabled=false}
   &SubUVAnimation(
    Evaluation Type=::NewEnumerator0
    Output Alternate Renderer Binding=true
    "Sprite Renderer "=DataInterface() {propertyValues {SpriteRenderer=&self:Flashing_Lights_Small001_0.NiagaraSpriteRendererProperties_2}}
    SubUV Animation Mode:ENiagara_SubUVLookupModeV2=::NewEnumerator2
    UseEndFrame=false
    UseStartFrame=false
   )
   &ShapeLocation(
    MaskSpawnGroup=false
    Offset=&MaskVectorBySpawnGroup(Masked Vector=<0 0 180>, Passthrough Vector=<0 0 180>)
    Offset Coordinate Space=::Local
    Offset Mode=::NewEnumerator0
    Override Seed=false
    Ring / Disc Distribution Mode:ENiagaraSphereDistributionMode=::NewEnumerator1
    Ring Radius=90
    Rotation Angle=0.5
    Rotation Axis=<0 0 1>
    Rotation Mode=::NewEnumerator1
    Shape Primitive=::NewEnumerator4
    U Position=&ReturnNormalizedExecIndex(Normalized Index Scale=0.5)
    UseOwnerScale=true
   )
   &AddVelocity(
    Velocity=&RandomRangeVector(
     Minimum=&MaskVector(Mask Source=1, Mask Parameter=$Emitter.SpawnGroup, Mask=0, Masked Vector=<-1 2 -5>, Passthrough Vector=<-1 2 -5>)
     Maximum=&MaskVector(Mask Source=1, Mask Parameter=$Emitter.SpawnGroup, Mask=0, Masked Vector=<1 2 -5>, Passthrough Vector=<1 2 -5>)
    )
    Velocity Speed Scale=0.01
   )
   &AddVelocity(
    Velocity=&Multiply_VectorByFloat(Vector=$Particles.ShapeLocation.ShapeTangent, Float=-1)
    Velocity Speed Scale=0.01
   ) {module=AddVelocity001}
   &ApplyInitialForces()
   /SetVariables(Particles.LightSize:Vector2f=<50 50>) {isSetParameters=true, module=SetVariables_#7}
  ]
  particleUpdateScript [
   &ParticleState()
   &ScaleSpriteSize(
    Scale Factor=&Vector2DFromCurve(
     Vector2Curve=DataInterface() {propertyValues {
       XCurve {
        defaultValue=0
        keys [
         {interpMode=RCIM_Cubic, tangentMode=RCTM_User, time=0, value=0}
         {arriveTangentWeight=0.1667, interpMode=RCIM_Cubic, tangentMode=RCTM_User, time=0.5, value=1}
         {arriveTangent=-3.3018, interpMode=RCIM_Cubic, leaveTangent=-3.3018, tangentMode=RCTM_User, time=0.8652, value=0.4388}
         <at 1 0>
        ]
       }
       YCurve {
        defaultValue=0
        keys [
         {interpMode=RCIM_Cubic, tangentMode=RCTM_User, time=0, value=0}
         {arriveTangent=2.4744, interpMode=RCIM_Cubic, leaveTangent=2.4744, tangentMode=RCTM_User, time=0.209, value=0.758}
         {arriveTangentWeight=0.1667, interpMode=RCIM_Cubic, tangentMode=RCTM_User, time=0.5, value=1}
         <at 1 0>
        ]
       }
      }}
     ScaleCurve=<1 1>
    )
    Scale Sprite Size Mode=::NewEnumerator2
   )
   &CurlNoiseForce(Bias Noise=true, Bias Noise Field:Vector3f=<0 1 0>, Noise Frequency=0, Noise Strength=0, Pan Noise=false)
   &SolveForcesAndVelocity(Clamp Velocity=false, Limit Acceleration=false)
   &ScaleColor(
    Scale Alpha=&FloatFromCurve(
     FloatCurve=DataInterface() {propertyValues {Curve {
        defaultValue=0
        keys [
         {interpMode=RCIM_Cubic, tangentMode=RCTM_User, time=0, value=0}
         {interpMode=RCIM_Cubic, tangentMode=RCTM_User, time=0.1374, value=1}
         {tangentMode=RCTM_User, time=1, value=0}
        ]
       }}}
     Scale Curve=&Multiply_Float(
      A=&MakeCustomFloatFromBool(
       Bool=&SetBoolByFloatComparison(
        A=&FN_PickFloatDependOnPlatform(
         TargetPlatform=DataInterface() {propertyValues {Platforms {
            qualityLevelMask=-30
            deviceProfileStates [
             {profileName=Windows, qualityLevelMask=0, setQualityLevelMask=2}
             {profileName=Mobile, qualityLevelMask=0, setQualityLevelMask=4294967295}
             {profileName=Mac, qualityLevelMask=0, setQualityLevelMask=2}
             {profileName=Linux, qualityLevelMask=0, setQualityLevelMask=2}
            ]
           }}}
         TargetPlatformActive=1
         TargetPlatformNoActive=0
        )
        B=0
       )
       True Float=0.5
       False Float=1
      )
      B=&Multiply_Float(
       A=&MakeCustomFloatFromBool(
        Bool=&SetBoolByIntComparison(A=$Engine.QualityLevel, B=0, Comparison Type=::NewEnumerator4)
        True Float=0.65
        False Float=1
       )
       B=&Multiply_Float(
        A=&RandomRangeFloat(Minimum=0.23, Maximum=1.1)
        B=&FNTimeOfDayScalar(Night Scale=0.4, Morning Scale=0.75, Evening Scale=0.75)
       )
      )
     )
    )
    ScaleA=true
    ScaleRGB=true
   )
   /SetVariables(
    Particles.FlareColor=&Multiply_LinearColorByFloat(Linear Color=$Particles.Color, Float=0.7)
    Particles.FlashSphereColor:LinearColor=&Multiply_LinearColorByFloat(Linear Color=$Particles.Color, Float=0.65)
   ) {isSetParameters=true}
  ]
  rendererClasses [&NiagaraSpriteRendererProperties]
  renderers [
   {rendererClass=&NiagaraSpriteRendererProperties}
   {rendererClass=&NiagaraSpriteRendererProperties, rendererIndex=1}
   {rendererClass=&NiagaraSpriteRendererProperties, rendererIndex=2}
  ]
 }
 Flashing_Lights_Small002 {
  bEnabled=false
  emitterUpdateScript [
   &EmitterState(Life Cycle Mode=::NewEnumerator1, Loop Duration=2.87, UseLoopDelay=true)
   &SpawnBurst_Instantaneous(Spawn Count=3, Use Spawn Probability=false, UseLoopCountLimit=false)
   &SpawnBurst_Instantaneous(Spawn Count=3, Spawn Time=1.0762, Use Spawn Probability=false, UseLoopCountLimit=false) {module=SpawnBurst_Instantaneous004}
   &SpawnBurst_Instantaneous(Spawn Count=4, Spawn Group=1, Spawn Time=1.3763, Use Spawn Probability=false, UseLoopCountLimit=false) {module=SpawnBurst_Instantaneous001}
   &SpawnBurst_Instantaneous(Spawn Count=4, Spawn Group=1, Spawn Time=1.7938, Use Spawn Probability=false, UseLoopCountLimit=false) {module=SpawnBurst_Instantaneous002}
   &SpawnBurst_Instantaneous(Spawn Count=3, Spawn Time=2.0938, Use Spawn Probability=false, UseLoopCountLimit=false) {module=SpawnBurst_Instantaneous003}
   &SpawnBurst_Instantaneous(Spawn Count=4, Spawn Group=1, Spawn Time=2.5113, Use Spawn Probability=false, UseLoopCountLimit=false) {module=SpawnBurst_Instantaneous005}
   &SpawnBurst_Instantaneous(Spawn Count=3, Spawn Time=2.87, Use Spawn Probability=false, UseLoopCountLimit=false) {module=SpawnBurst_Instantaneous006}
   &SpawnParticlesInGrid(X Count=5, Y Count=0, Z Count=3) {enabled=false}
   &SpawnParticlesInGrid(Spawn Group=1, X Count=5, Y Count=0, Z Count=3) {enabled=false, module=SpawnParticlesInGrid001}
  ]
  particleSpawnScript [
   &InitializeParticle(
    Color=<rgba 0.8482 0.8345 1 0.8>
    Lifetime=&RandomRangeFloat(Minimum=0.55, Maximum=0.6)
    Sprite Size=&RandomRangeVector2D(Minimum=<60 225>, Maximum=<80 250>)
    Sprite Size Mode=::NewEnumerator4
    UsePositionOffset=false
    Write Material Random=true
   )
   &GridLocation(
    Module.Jitter Inside Grid Cells=false
    Module.Override Random Seed=false
    Offset:Vector3f=&MaskVector(Mask Source=0, Mask Parameter=?, Mask=0, Masked Vector=<6 -50 18>, Passthrough Vector=<-6 -50 18>)
    XYZ Dimensions:Vector3f=<10 1 10>
   ) {enabled=false}
   &SubUVAnimation(
    Evaluation Type=::NewEnumerator0
    Output Alternate Renderer Binding=true
    "Sprite Renderer "=DataInterface() {propertyValues {SpriteRenderer=&self:Flashing_Lights_Small002_1.NiagaraSpriteRendererProperties_2}}
    SubUV Animation Mode:ENiagara_SubUVLookupModeV2=::NewEnumerator2
    UseEndFrame=false
    UseStartFrame=false
   )
   &ShapeLocation(
    MaskSpawnGroup=false
    Offset=&MaskVectorBySpawnGroup(Masked Vector=<0 0 180>, Passthrough Vector=<0 0 170>)
    Offset Coordinate Space=::Local
    Offset Mode=::NewEnumerator0
    Override Seed=false
    Radius Position:NiagaraFloat=&RandomRangeFloat(Minimum=0.9, Maximum=1)
    Ring / Disc Distribution Mode:ENiagaraSphereDistributionMode=::NewEnumerator1
    Ring Radius=75
    Rotation Angle=0.5
    Rotation Axis=<0 0 1>
    Rotation Mode=::NewEnumerator1
    Shape Primitive=::NewEnumerator4
    U Position=&ReturnNormalizedExecIndex(Normalized Index Scale=0.5)
    UseOwnerScale=true
   )
   &AddVelocity(
    Velocity=&RandomRangeVector(
     Minimum=&MaskVector(Mask Source=1, Mask Parameter=$Emitter.SpawnGroup, Mask=0, Masked Vector=<1 2 -5>, Passthrough Vector=<1 2 -5>)
     Maximum=&MaskVector(Mask Source=1, Mask Parameter=$Emitter.SpawnGroup, Mask=0, Masked Vector=<-1 2 -5>, Passthrough Vector=<-1 2 -5>)
    )
    Velocity Speed Scale=0.01
   )
   &AddVelocity(
    Velocity=&Multiply_VectorByFloat(Vector=$Particles.ShapeLocation.ShapeTangent, Float=1)
    Velocity Speed Scale=0.01
   ) {module=AddVelocity001}
   &ApplyInitialForces()
   /SetVariables(Particles.LightSize:Vector2f=<50 50>) {isSetParameters=true, module=SetVariables_#7}
  ]
  particleUpdateScript [
   &ParticleState()
   &ScaleSpriteSize(
    Scale Factor=&Vector2DFromCurve(
     Vector2Curve=DataInterface() {propertyValues {
       XCurve {
        defaultValue=0
        keys [
         {interpMode=RCIM_Cubic, tangentMode=RCTM_User, time=0, value=0}
         {arriveTangentWeight=0.1667, interpMode=RCIM_Cubic, tangentMode=RCTM_User, time=0.5, value=1}
         {arriveTangent=-3.3018, interpMode=RCIM_Cubic, leaveTangent=-3.3018, tangentMode=RCTM_User, time=0.8652, value=0.4388}
         <at 1 0>
        ]
       }
       YCurve {
        defaultValue=0
        keys [
         {interpMode=RCIM_Cubic, tangentMode=RCTM_User, time=0, value=0}
         {arriveTangent=2.4744, interpMode=RCIM_Cubic, leaveTangent=2.4744, tangentMode=RCTM_User, time=0.209, value=0.758}
         {arriveTangentWeight=0.1667, interpMode=RCIM_Cubic, tangentMode=RCTM_User, time=0.5, value=1}
         <at 1 0>
        ]
       }
      }}
     ScaleCurve=<1 1>
    )
    Scale Sprite Size Mode=::NewEnumerator2
   )
   &CurlNoiseForce(Bias Noise=true, Bias Noise Field:Vector3f=<0 1 0>, Noise Frequency=0, Noise Strength=0, Pan Noise=false)
   &SolveForcesAndVelocity(Clamp Velocity=false, Limit Acceleration=false)
   &ScaleColor(
    Scale Alpha=&FloatFromCurve(
     FloatCurve=DataInterface() {propertyValues {Curve {
        defaultValue=0
        keys [
         {interpMode=RCIM_Cubic, tangentMode=RCTM_User, time=0, value=0}
         {interpMode=RCIM_Cubic, tangentMode=RCTM_User, time=0.1374, value=1}
         {tangentMode=RCTM_User, time=1, value=0}
        ]
       }}}
     Scale Curve=&Multiply_Float(
      A=&MakeCustomFloatFromBool(
       Bool=&SetBoolByFloatComparison(
        A=&FN_PickFloatDependOnPlatform(
         TargetPlatform=DataInterface() {propertyValues {Platforms {
            qualityLevelMask=-30
            deviceProfileStates [
             {profileName=Windows, qualityLevelMask=0, setQualityLevelMask=2}
             {profileName=Mobile, qualityLevelMask=0, setQualityLevelMask=4294967295}
             {profileName=Mac, qualityLevelMask=0, setQualityLevelMask=2}
             {profileName=Linux, qualityLevelMask=0, setQualityLevelMask=2}
            ]
           }}}
         TargetPlatformActive=1
         TargetPlatformNoActive=0
        )
        B=0
       )
       True Float=0.5
       False Float=1
      )
      B=&Multiply_Float(
       A=&MakeCustomFloatFromBool(
        Bool=&SetBoolByIntComparison(A=$Engine.QualityLevel, B=0, Comparison Type=::NewEnumerator4)
        True Float=0.65
        False Float=1
       )
       B=&Multiply_Float(
        A=&RandomRangeFloat(Minimum=0.23, Maximum=1.1)
        B=&FNTimeOfDayScalar(Night Scale=0.4, Morning Scale=0.75, Evening Scale=0.75)
       )
      )
     )
    )
    ScaleA=true
    ScaleRGB=true
   )
   /SetVariables(
    Particles.FlareColor=&Multiply_LinearColorByFloat(Linear Color=$Particles.Color, Float=0.7)
    Particles.FlashSphereColor:LinearColor=&Multiply_LinearColorByFloat(Linear Color=$Particles.Color, Float=0.65)
   ) {isSetParameters=true, module=SetVariables_#8}
  ]
  rendererClasses [&NiagaraSpriteRendererProperties]
  renderers [
   {rendererClass=&NiagaraSpriteRendererProperties}
   {rendererClass=&NiagaraSpriteRendererProperties, rendererIndex=1}
   {rendererClass=&NiagaraSpriteRendererProperties, rendererIndex=2}
  ]
 }
 GroundGlow {
  emitterUpdateScript [
   &EmitterState(
    Delay First Loop Only=true
    Life Cycle Mode=::NewEnumerator1
    Loop Behavior=::NewEnumerator2
    Loop Count=4
    Loop Duration=1.435
    UseLoopDelay=true
   )
   &SpawnBurst_Instantaneous(Use Spawn Probability=false, UseLoopCountLimit=false)
   &SpawnBurst_Instantaneous(Spawn Time=0.3587, Use Spawn Probability=false, UseLoopCountLimit=false) {module=SpawnBurst_Instantaneous001}
   &SpawnBurst_Instantaneous(Spawn Time=1.0762, Use Spawn Probability=false, UseLoopCountLimit=false) {module=SpawnBurst_Instantaneous002}
  ]
  particleSpawnScript [
   &InitializeParticle(
    Color=<rgba 0.8482 0.8345 1 0.3>
    Lifetime=0.8
    Mesh Scale=<2.5 2.5 1>
    Mesh Scale Mode=::NewEnumerator4
    Position Offset=<0 0 3>
    Sprite Size Mode=::NewEnumerator3
    Uniform Sprite Size=80
    UsePositionOffset=true
    Write Material Random=true
   )
   /SetVariables(Particles.Glow DecalSize:Vector3f=<550 500 500>) {isSetParameters=true, module=SetVariables_#4}
   /SetVariables(Particles.DecalSize:Vector3f=<350 350 350>) {isSetParameters=true, module=SetVariables_#5}
   &SpriteFacingAndAlignment(Sprite Facing=<0 0 1>, WriteAlignment=true, WriteFacing=true)
  ]
  particleUpdateScript [
   &ParticleState()
   &SpriteRotationRate(Rotation Rate=&RandomRangeFloat(Minimum=50, Maximum=70))
   &ScaleSpriteSize(
    Uniform Curve Scale=1.5
    Uniform Curve Sprite Scale=DataInterface() {propertyValues {Curve {
       defaultValue=0
       keys [
        {arriveTangent=4.1785, interpMode=RCIM_Cubic, leaveTangent=4.1785, tangentMode=RCTM_User, time=0, value=0.2}
        {arriveTangent=1.1663, interpMode=RCIM_Cubic, leaveTangent=1.1663, tangentMode=RCTM_User, time=0.2042, value=0.7}
        {arriveTangent=0.0224, interpMode=RCIM_Cubic, leaveTangent=0.0224, tangentMode=RCTM_User, time=1, value=1}
       ]
      }}}
   )
   &SolveForcesAndVelocity(Clamp Velocity=false, Limit Acceleration=false)
   &ScaleColor(
    Scale Alpha=&Multiply_Float(
     A=&FloatFromCurve(
      FloatCurve=DataInterface() {propertyValues {Curve {
         defaultValue=0.9847
         keys [
          {tangentMode=RCTM_User, time=0.1, value=1}
          {arriveTangent=-0.6538, interpMode=RCIM_Cubic, leaveTangent=-0.6538, tangentMode=RCTM_User, time=0.25, value=1.025}
          {arriveTangent=-0.46, interpMode=RCIM_Cubic, leaveTangent=-0.46, tangentMode=RCTM_User, time=1, value=0}
         ]
        }}}
      Scale Curve=&Multiply_Float(A=0.5, B=&FNTimeOfDayScalar(Night Scale=0.8, Morning Scale=0.8, Evening Scale=0.8))
     )
     B=1
    )
    ScaleA=true
    ScaleRGB=true
   )
   &BlendSpace(Rotation Contribution=0)
   /SetVariables(Particles.Glow DecalColor:LinearColor=&MakeLinearColorFromVectorAndFloat(
     "Vector (RGB)"=&MakeVectorFromLinearColorRGB(Color=<rgba 1 1 1 1>)
     "Float (Alpha)"=&Multiply_Float(
      A=&FloatFromCurve(
       FloatCurve=DataInterface() {propertyValues {Curve {keys [
           {arriveTangent=-2.4342, interpMode=RCIM_Cubic, leaveTangent=-2.4342, tangentMode=RCTM_User, time=0, value=1}
           {arriveTangent=-0.3806, interpMode=RCIM_Cubic, leaveTangent=-0.3806, tangentMode=RCTM_User, time=1, value=0}
          ]}}}
       Scale Curve=&Multiply_Float(
        A=&FN_FrontEndFloatModifier(Gameplay=0.13, Front End=0.15)
        B=&FNTimeOfDayScalar(Night Scale=0.55, Morning Scale=0.8, Evening Scale=0.8)
       )
      )
      B=1
     )
    )) {isSetParameters=true, module=SetVariables_#6}
  ]
  rendererClasses [&NiagaraDecalRendererProperties, &NiagaraSpriteRendererProperties]
  renderers [
   {rendererClass=&NiagaraDecalRendererProperties, rendererIndex=0}
   {rendererClass=&NiagaraSpriteRendererProperties, rendererIndex=1}
  ]
 }
 Ground_Light {rendererClasses [], renderers []}
 Light_Left {rendererClasses [], renderers []}
 Light_Right {rendererClasses [], renderers []}
}
summary {
 systemName=NS_Emote_NoiseGlueGustSmallLight
 emitters [
  {emitterName=Light_Right, rendererClasses []}
  {emitterName=Back_Light, rendererClasses []}
  {emitterName=Light_Left, rendererClasses []}
  {emitterName=Flashing_Lights_Small, rendererClasses [&NiagaraSpriteRendererProperties]}
  {emitterName=Flashing_Lights_Small001, rendererClasses [&NiagaraSpriteRendererProperties]}
  {bEnabled=false, emitterName=Flashing_Lights_Small002, rendererClasses [&NiagaraSpriteRendererProperties]}
  {emitterName=Decal, rendererClasses [&NiagaraDecalRendererProperties, &NiagaraSpriteRendererProperties]}
  {emitterName=GroundGlow, rendererClasses [&NiagaraDecalRendererProperties, &NiagaraSpriteRendererProperties]}
  {emitterName=Ground_Light, rendererClasses []}
 ]
 userVariables [
  User.bIsFrontEnd {type=NiagaraBool}
  User.bIsFrontEndPreview {type=NiagaraBool}
  User.Fade {type=NiagaraFloat, defaultValue=0, description=""}
  User.Pyrotechnics {type=NiagaraFloat, defaultValue=0.5, description=""}
  User.SquareRingColor {type=LinearColor, description="", defaultValue {struct=LinearColor, value=<rgba 0 0.3611 0.7083 1>}}
 ]
}
